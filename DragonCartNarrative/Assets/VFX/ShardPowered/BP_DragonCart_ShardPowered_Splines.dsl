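; Powered Dragon Cart shard VFX actor that drives a Niagara shard heart effect along three splines, with optional audio that follows the local player.
; Construction and BeginPlay pass the North, West and East splines to Niagara.
; Once game state is ready, a half-second timer moves the audio to the spline point nearest the pawn, fading by range.
; The search tries the last matched spline first, then takes the first in range.
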(event EventBeginPlay
  (bind _returnvalue (Game|GetPlayerController 0))
  (bind _returnvalue_2 (Pawn|GetControlledPawn _returnvalue))
  (CallFunction|SetUpFXSplines)
  (Class|FortAsyncActionGameStateReady|GameStateReadyAsync
    (:OnReady
      (Utilities|IsValid _returnvalue_2
        (:Is%20Valid
          (bind _asfort_player_pawn (Utilities|Casting|CastToFortPlayerPawn _returnvalue_2))
          (Variables|Default|SetLocalPawn _asfort_player_pawn)
          (Variables|Default|GetShouldSplineAudioPlay
            (:then
              (CallFunction|ActivateSplineAudio))
            (:else)))))))

(event Custom|UpdateSplineAudio
  (bind _splineaudio (Variables|Default|GetSplineAudio))
  (bind (_closestpointonspline _point_in_range) (CallFunction|FindPointClosesttoPawn))
  (if _point_in_range
    (if (not (Audio|Components|Audio|IsPlaying _splineaudio))
      (Audio|Components|Audio|FadeIn _splineaudio 1.0 1.0 0.0 "Linear" :!class /Script/Engine.AudioComponent))
    (Transformation|SetWorldLocation _splineaudio _closestpointonspline false false)
    (else
      (bind _splineaudio_2 (Variables|Default|GetSplineAudio)
        (:then
          (Audio|Components|Audio|FadeOut _splineaudio_2 1.0 0.0 "Linear" :!class /Script/Engine.AudioComponent))
        (:else)))))

(event Custom|ActivateSplineAudio
  (Utilities|Time|SetTimerbyEvent (EventDispatchers|CreateEvent :!event Custom|UpdateSplineAudio) 0.5 true))

(fn FindPointClosestToPawn
  (bind _splineaudioactivationrange (Variables|Default|GetSplineAudioActivationRange))
  (bind _splineaudioactivationrange_2 (Variables|Default|GetSplineAudioActivationRange))
  (bind _splineaudioactivationrange_3 (Variables|Default|GetSplineAudioActivationRange))
  (bind _splineeast (Variables|Default|GetSplineEast))
  (bind _splineaudioactivationrange_4 (Variables|Default|GetSplineAudioActivationRange))
  (bind _splinenorth (Variables|Default|GetSplineNorth))
  (bind _splinewest (Variables|Default|GetSplineWest))
  (bind _localpawn (Variables|Default|GetLocalPawn)
    (:then
      (bind _lastclosestspline (Variables|Default|GetLastClosestSpline)
        (:then
          (LocalVariables|SetLastSplineClosestPoint (Spline|FindLocationClosesttoWorldLocation _lastclosestspline (Transformation|GetActorLocation _localpawn) "World"))
          (if (<= (Math|Vector|Distance(Vector) (Spline|FindLocationClosesttoWorldLocation _lastclosestspline (Transformation|GetActorLocation _localpawn) "World") (Transformation|GetActorLocation _localpawn)) _splineaudioactivationrange)
            (return (LocalVariables|GetLastSplineClosestPoint) (<= (Math|Vector|Distance(Vector) (Spline|FindLocationClosesttoWorldLocation _lastclosestspline (Transformation|GetActorLocation _localpawn) "World") (Transformation|GetActorLocation _localpawn)) _splineaudioactivationrange))
            (else
              _)))
        (:else
          _))
      (LocalVariables|SetEastSplineClosestPoint (Spline|FindLocationClosesttoWorldLocation _splineeast (Transformation|GetActorLocation _localpawn) "World"))
      (bind _output_get (LocalVariables|SetDistancEastSpline (Math|Vector|Distance(Vector) (Spline|FindLocationClosesttoWorldLocation _splineeast (Transformation|GetActorLocation _localpawn) "World") (Transformation|GetActorLocation _localpawn))))
      (if (<= _output_get _splineaudioactivationrange_4)
        (Variables|Default|SetLastClosestSpline (Variables|Default|GetSplineEast))
        (return (LocalVariables|GetEastSplineClosestPoint) (<= _output_get _splineaudioactivationrange_4))
        (else
          (LocalVariables|SetNorthSplineClosestPoint (Spline|FindLocationClosesttoWorldLocation _splinenorth (Transformation|GetActorLocation _localpawn) "World"))
          (bind _output_get_2 (LocalVariables|SetDistanceNorthSpline (Math|Vector|Distance(Vector) (Spline|FindLocationClosesttoWorldLocation _splinenorth (Transformation|GetActorLocation _localpawn) "World") (Transformation|GetActorLocation _localpawn))))
          (if (<= _output_get_2 _splineaudioactivationrange_3)
            (Variables|Default|SetLastClosestSpline (Variables|Default|GetSplineNorth))
            (return (LocalVariables|GetNorthSplineClosestPoint) (<= _output_get_2 _splineaudioactivationrange_3))
            (else
              (LocalVariables|SetWestSplineClosestPoint (Spline|FindLocationClosesttoWorldLocation _splinewest (Transformation|GetActorLocation _localpawn) "World"))
              (bind _output_get_3 (LocalVariables|SetDistanceWestSpline (Math|Vector|Distance(Vector) (Spline|FindLocationClosesttoWorldLocation _splinewest (Transformation|GetActorLocation _localpawn) "World") (Transformation|GetActorLocation _localpawn))))
              (if (<= _output_get_3 _splineaudioactivationrange_2)
                (Variables|Default|SetLastClosestSpline (Variables|Default|GetSplineWest))
                (return (LocalVariables|GetWestSplineClosestPoint) (<= _output_get_3 _splineaudioactivationrange_2))))))))
    (:else
      (return (Transformation|GetWorldLocation (Variables|Default|GetCenter)) true))))

(fn SearchForPawn
  (bind _returnvalue (Game|GetPlayerController 0))
  (bind _returnvalue_2 (Pawn|GetControlledPawn _returnvalue))
  (Utilities|IsValid _returnvalue_2
    (:Is%20Valid
      (bind _asfort_player_pawn (Utilities|Casting|CastToFortPlayerPawn _returnvalue_2))
      (Variables|Default|SetLocalPawn _asfort_player_pawn)
      (Utilities|Time|SetTimerbyEvent (EventDispatchers|CreateEvent :!event UpdateSplineAudio) 0.5 true))))

(fn SetUpFXSplines
  (bind _ns_dragoncart_shard_heart (Variables|Default|GetNS_DragonCart_Shard_Heart))
  (Niagara|SetNiagaraVariable(Object) _ns_dragoncart_shard_heart "Spline1" (Variables|Default|GetSplineNorth))
  (Niagara|SetNiagaraVariable(Object) _ns_dragoncart_shard_heart "Spline2" (Variables|Default|GetSplineWest))
  (Niagara|SetNiagaraVariable(Object) _ns_dragoncart_shard_heart "Spline3" (Variables|Default|GetSplineEast)))

(fn ConstructionScript
  (CallFunction|SetUpFXSplines))
