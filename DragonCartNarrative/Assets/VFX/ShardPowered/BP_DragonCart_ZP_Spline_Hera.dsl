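; Spline actor that renders a shard-powered energy-flow effect with optional proximity audio that follows the local player.
; The construction script lays spline mesh segments along SplineNorth, sharing one dynamic material.
; Once game state is ready, BeginPlay caches the pawn and, if audio is enabled, starts a half-second timer.
; Each tick moves the audio to the pawn's nearest point on LastClosestSpline, fading it in or out by range.
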(event EventBeginPlay
  (bind _returnvalue (Game|GetPlayerController 0))
  (bind _returnvalue_2 (Pawn|GetControlledPawn _returnvalue))
  (Class|FortAsyncActionGameStateReady|GameStateReadyAsync
    (:OnReady
      (Utilities|IsValid _returnvalue_2
        (:Is%20Valid
          (bind _asfort_player_pawn (Utilities|Casting|CastToFortPlayerPawn _returnvalue_2))
          (Variables|Default|SetLocalPawn _asfort_player_pawn)
          (Variables|Default|GetShouldSplineAudioPlay
            (:then
              (CallFunction|ActivateSplineAudio))
            (:else)))))))

(event Custom|UpdateSplineAudio
  (bind _splineaudio (Variables|Default|GetSplineAudio))
  (bind (_closestpointonspline _point_in_range) (CallFunction|FindPointClosesttoPawn))
  (if _point_in_range
    (if (not (Audio|Components|Audio|IsPlaying _splineaudio))
      (Audio|Components|Audio|FadeIn _splineaudio 1.0 1.0 0.0 "Linear" :!class /Script/Engine.AudioComponent))
    (Transformation|SetWorldLocation _splineaudio _closestpointonspline false false)
    (else
      (bind _splineaudio_2 (Variables|Default|GetSplineAudio)
        (:then
          (Audio|Components|Audio|FadeOut _splineaudio_2 1.0 0.0 "Linear" :!class /Script/Engine.AudioComponent))
        (:else)))))

(event Custom|ActivateSplineAudio
  (Utilities|Time|SetTimerbyEvent (EventDispatchers|CreateEvent :!event Custom|UpdateSplineAudio) 0.5 true))

(fn FindPointClosestToPawn
  (bind _splineaudioactivationrange (Variables|Default|GetSplineAudioActivationRange))
  (bind _localpawn (Variables|Default|GetLocalPawn)
    (:then
      (bind _lastclosestspline (Variables|Default|GetLastClosestSpline)
        (:then
          (LocalVariables|SetLastSplineClosestPoint (Spline|FindLocationClosesttoWorldLocation _lastclosestspline (Transformation|GetActorLocation _localpawn) "World"))
          (if (<= (Math|Vector|Distance(Vector) (Spline|FindLocationClosesttoWorldLocation _lastclosestspline (Transformation|GetActorLocation _localpawn) "World") (Transformation|GetActorLocation _localpawn)) _splineaudioactivationrange)
            (return (LocalVariables|GetLastSplineClosestPoint) (<= (Math|Vector|Distance(Vector) (Spline|FindLocationClosesttoWorldLocation _lastclosestspline (Transformation|GetActorLocation _localpawn) "World") (Transformation|GetActorLocation _localpawn)) _splineaudioactivationrange))))
        (:else)))
    (:else
      (return (Transformation|GetWorldLocation (Variables|Default|GetCenter)) true))))

(fn SearchForPawn
  (bind _returnvalue (Game|GetPlayerController 0))
  (bind _returnvalue_2 (Pawn|GetControlledPawn _returnvalue))
  (Utilities|IsValid _returnvalue_2
    (:Is%20Valid
      (bind _asfort_player_pawn (Utilities|Casting|CastToFortPlayerPawn _returnvalue_2))
      (Variables|Default|SetLocalPawn _asfort_player_pawn)
      (Utilities|Time|SetTimerbyEvent (EventDispatchers|CreateEvent :!event UpdateSplineAudio) 0.5 true))))

(fn SetSplineOpacity
  (params (Visibility :type Boolean))
  (Rendering|Material|SetScalarParameterValue (Variables|Default|GetSplineMeshMaterial) "OpacityMult" (Variables|Default|GetSplineMeshOpacity) :!class /Script/Engine.MaterialInstanceDynamic)
  (for _array_element (Variables|Default|GetSplineMeshes)
    (Rendering|SetVisibility _array_element Visibility false)))

(fn SetupSplineMeshes
  (params (Spline :type Spline%20Component%20Object%20Reference))
  (bind _returnvalue (Rendering|Material|CreateDynamicMaterialInstance "/DragonCartNarrative/Assets/VFX/ShardPowered/M_TechHQ_Energy_Flow_Spline_Awakening_Inst.M_TechHQ_Energy_Flow_Spline_Awakening_Inst" "None" "None" :!class /Script/Engine.KismetMaterialLibrary))
  (bind _output_get (Variables|Default|SetSplineMeshMaterial _returnvalue))
  (Rendering|Material|SetScalarParameterValue _output_get "SplineWidth" 2.5e+02 :!class /Script/Engine.MaterialInstanceDynamic)
  (Rendering|Material|SetScalarParameterValue _output_get "PanSpeed" -0.12 :!class /Script/Engine.MaterialInstanceDynamic)
  (for _index (range (+ (- (Spline|GetNumberOfSplinePoints Spline) 2) 1))
    (bind _returnvalue_2 (AddComponent|NODE_AddSplineMeshComponent-0))
    (Rendering|Material|SetMaterial _returnvalue_2 0 (Variables|Default|GetSplineMeshMaterial))
    (Utilities|Array|Add (Variables|Default|GetSplineMeshes) _returnvalue_2)
    (bind (_returnvalue_inputkey _returnvalue_position _returnvalue_arrivetangent _returnvalue_leavetangent _returnvalue_rotation _returnvalue_scale) (Utilities|Struct|BreakSplinePoint (Spline|GetSplinePointAt Spline _index "Local")))
    (SplineMesh|SetStartPosition _returnvalue_2 _returnvalue_position false)
    (SplineMesh|SetStartTangent _returnvalue_2 _returnvalue_arrivetangent false :!class /Script/Engine.SplineMeshComponent)
    (bind (_returnvalue_inputkey_2 _returnvalue_position_2 _returnvalue_arrivetangent_2 _returnvalue_leavetangent_2 _returnvalue_rotation_2 _returnvalue_scale_2) (Utilities|Struct|BreakSplinePoint (Spline|GetSplinePointAt Spline (+ _index 1) "Local")))
    (SplineMesh|SetEndPosition _returnvalue_2 _returnvalue_position_2 false)
    (SplineMesh|SetEndTangent _returnvalue_2 _returnvalue_leavetangent_2 false :!class /Script/Engine.SplineMeshComponent)
    (SplineMesh|SetStartScale _returnvalue_2 (Math|Vector2D|MakeVector2D (.y _returnvalue_scale) (.z _returnvalue_scale)) true :!class /Script/Engine.SplineMeshComponent)
    (SplineMesh|SetEndScale _returnvalue_2 (Math|Vector2D|MakeVector2D (.y _returnvalue_scale_2) (.z _returnvalue_scale_2)) true :!class /Script/Engine.SplineMeshComponent)
    (if (== _index 0)
      (SplineMesh|SetStartScale _returnvalue_2 "(X=1.000000,Y=1.000000)" true :!class /Script/Engine.SplineMeshComponent))))

(fn ConstructionScript
  (CallFunction|SetupSplineMeshes (Variables|Default|GetSplineNorth)))
