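# Material function that computes world position offset for opening the ZerO gate portal as the camera approaches.
# Open Amount rises as camera distance to a Dynamic Parameter XYZ position shrinks, using that parameter's OpenRadius and an inner radius offset.
# A powered sphere mask on UV0 V shapes the opening, while UV1 U, remapped around zero, pushes vertices apart along local Y.
# These terms, scaled by PortalOpen and Open Amount, form PortalWPO; OpenAmount is also output.
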MaterialFunction /DragonCartNarrative/Assets/VFX/ShardPowered/MF_ZerOGate_OpenWPO
graph {nodes [
  Constant()
  n1 = ComponentMask(None=TextureCoordinate()) {g=true}
  n2 = ScalarParameter() {defaultValue=1.2, group=OpeningWPO, parameterName=SphereCenterPower}
  n3 = ScalarParameter() {defaultValue=0.5, group=OpeningWPO, parameterName=SphereCenter}
  n4 = ScalarParameter() {defaultValue=0.5, group=OpeningWPO, parameterName=SphereRadius}
  n5 = Power(
   Base=SphereMask(
    A=Power(Base=@n1, Exp=@n2)
    B=FunctionInput(Preview=@n3) {bUsePreviewValueAsDefault=true, inputName=SphereCenter, inputType=FunctionInput_Scalar}
    Radius=FunctionInput(Preview=@n4) {bUsePreviewValueAsDefault=true, inputName=SphereRadius, inputType=FunctionInput_Scalar}
   ) {hardnessPercent=0}
   Exp=ScalarParameter() {defaultValue=0.7, group=OpeningWPO, parameterName=SphereOpenPower}
  ) {constExponent=0.6}
  n6 = TextureCoordinate() {coordinateIndex=1}
  n7 = Multiply(A=Subtract(A=ComponentMask(None=@n6) {r=true}) {constB=0.5}) {constB=2}
  n8 = Multiply(A=@n7, B=Transform(None=Constant3Vector() {constant=<rgba 0 1 0 0>}) {transformSourceType=TRANSFORMSOURCE_Local})
  n9 = ScalarParameter() {defaultValue=140, group=OpeningWPO, parameterName=PortalOpen}
  n10 = NamedRerouteUsage()
  FunctionOutput(None=Multiply(A=Multiply(A=@n5, B=@n8), B=Multiply(A=@n9, B=@n10))) {outputName=PortalWPO}
  FunctionOutput(None=@n10) {outputName=OpenAmount}
  MaterialFunctionCall() {materialFunction=&DebugTimeSine}
  Multiply(A=@n7, B=MaterialFunctionCall() {materialFunction=&TangentBasis}.X U)
  n11 = CameraPositionWS()
  n12 = DynamicParameter() {defaultValue=<rgba 1000 1 1 1000>, paramNames [X,Y,Z,OpenRadius]}
  n13 = Divide(A=Add(A=Distance(A=@n11, B=@n12.RGB), B=ScalarParameter() {defaultValue=-500, parameterName=InnerOpenRadius}), B=@n12.OpenRadius)
  NamedRerouteDeclaration(None=OneMinus(None=Saturate(None=@n13))) {name=Open Amount, nodeColor=<rgba 1 0 0.8437 1>}
  Saturate(None=@n5)
 ]}
