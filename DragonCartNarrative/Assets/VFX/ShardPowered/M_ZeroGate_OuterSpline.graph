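# Translucent, unlit, two-sided material for Niagara mesh particles, drawing the outer energy splines of the Zero Gate portal effect.
# World position offset combines SplineThicken width, a travelling sine wave and portal motion from the OpenWPO function.
# Panned caustics distort two tinted star-line layers over a background gradient, with vertex color offsetting pan timing.
# Opacity comes from those textures, shaped by UV end fades, near-camera fade, depth fade and particle alpha.
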Material /DragonCartNarrative/Assets/VFX/ShardPowered/M_ZeroGate_OuterSpline
graph {
 properties {
  bAutomaticallySetUsageInEditor=false
  bCastRayTracedShadows=false
  bUsedWithNiagaraMeshParticles=true
  blendMode=BLEND_Translucent
  ditherOpacityMask=true
  shadingModel=MSM_Unlit
  twoSided=true
  usedShadingModels=Unlit
 }
 nodes [
  n1 = ScalarParameter() {defaultValue=160, group=Spline Thicken, parameterName=SplineWidth}
  n2 = Multiply(A=@n1, B=ScalarParameter() {defaultValue=0.25, group=Spline Thicken, parameterName=WidthTipMult})
  n3 = NamedRerouteDeclaration(None=@n1) {name=SplineWidth, nodeColor=<rgba 0.4062 0 1 1>}
  n4 = NamedRerouteDeclaration(None=TextureCoordinate()) {name=UV0, nodeColor=<rgba 0 0.7656 1 1>}
  n5 = ScalarParameter() {defaultValue=0.5, group=End Fade, parameterName=EndFadeRadius_Y}
  n6 = SphereMask(
   A=ComponentMask(None=@n4) {g=true}
   B=@n5
   Radius=@n5
   Hardness=ScalarParameter() {defaultValue=0.5, group=End Fade, parameterName=EndFadeHardness_Y}
  ) {hardnessPercent=0.25}
  n7 = OneMinus(None=@n6)
  n8 = MaterialFunctionCall() {materialFunction=&TangentBasis}
  n9 = ComponentMask(None=NamedRerouteUsage()) {g=true}
  n10 = Add(
   A=Multiply(A=Time(), B=ScalarParameter() {defaultValue=-0.07, group=Sine Wave, parameterName=WaveSpeed})
   B=Divide(A=@n9, B=ScalarParameter() {defaultValue=0.9, group=Sine Wave, parameterName=WaveLength})
  )
  n11 = Multiply(A=LinearInterpolate(A=@n8.X U, B=Multiply(A=@n8.Z W) {constB=2.5}, Alpha=NamedRerouteUsage()), B=Sine(None=@n10))
  n12 = NamedRerouteUsage()
  n13 = Constant() {r=0.5}
  n14 = Add(A=SphereMask(A=ComponentMask(None=@n12) {g=true}, B=@n13, Radius=@n13) {hardnessPercent=0})
  n15 = LinearInterpolate(Alpha=NamedRerouteUsage()) {constA=1, constB=0.4}
  n16 = LinearInterpolate(Alpha=@n9) {constA=5, constB=0.2}
  n17 = Multiply(A=Multiply(A=Add(A=@n14, B=@n15), B=Saturate(None=@n16)), B=ScalarParameter() {defaultValue=15, group=Sine Wave, parameterName=WaveScale})
  n18 = MaterialFunctionCall!(SphereRadius, SphereCenter) {materialFunction=&/MF_ZerOGate_OpenWPO}
  Add(
   A=Add(
    A=MaterialFunctionCall(WidthBase=@n2, WidthTip=@n3, UVs for Thickness=@n7) {materialFunction=&SplineThicken}.WorldPositionOffset
    B=Multiply(A=@n11, B=@n17)
   ) {constB=0}
   B=@n18.PortalWPO
  )
  n19 = VectorParameter() {group=Texture Parameter, parameterName=Tile1, defaultValue=<rgba 0.15 4 0 0>}
  n20 = Multiply(A=ScalarParameter() {defaultValue=0.02, parameterName=SplineWidthTileMult}, B=NamedRerouteUsage()) {constB=0.1}
  n21 = Multiply(A=NamedRerouteUsage(), B=AppendVector(A=Multiply(A=@n19.R, B=@n20), B=@n19.G))
  n22 = VectorParameter() {group=Texture Parameter, parameterName=Tile2, defaultValue=<rgba 1 0.1 0 0>}
  n23 = AppendVector(A=@n22.R, B=@n22.R)
  n24 = VectorParameter() {group=Texture Parameter, parameterName=PanDirection2, defaultValue=<rgba 0.04 0.1 0 0>}
  n25 = VertexColor()
  n26 = Time()
  n27 = Add(A=@n25.G, B=Multiply(A=Multiply(A=@n25.B, B=@n26), B=ScalarParameter() {defaultValue=-0.5, parameterName=PanSpeed}))
  n28 = TextureSample(UVs=Panner(Coordinate=Multiply(A=@n21, B=@n23), Time=@n27, Speed=AppendVector(A=@n24.R, B=@n24.R)) {speedX=1, speedY=1}) {texture=/Game/Effects/Fort_Effects/Textures/Tile/T_Caustics_Chromatic}
  n29 = Multiply(
   A=ComponentMask(None=@n28.RGB) {g=true, r=true}
   B=ScalarParameter() {defaultValue=0.05, group=Texture Parameter, parameterName=DistortionIntensity}
  ) {constB=0.05}
  n30 = VectorParameter() {group=Texture Parameter, parameterName=PanDirection1, defaultValue=<rgba 0.08 1 0 0>}
  n31 = AppendVector(A=@n30.R, B=@n30.R)
  n32 = TextureSample(UVs=Panner(Coordinate=Add(A=@n21, B=@n29), Time=@n27, Speed=@n31) {speedX=1, speedY=1}) {texture=/Game/Athena/Environments/Nexus/Textures/T_ZeroPoint_Line_Stars}
  n33 = Multiply(A=VectorParameter() {group=Color, parameterName=PrimaryTint, defaultValue=<rgba 1 1 1 1>}.RGB, B=@n32.RGB) {constB=0.5}
  n34 = Multiply(A=@n21) {constB=2}
  n35 = VectorParameter() {group=Texture Parameter, parameterName=PanDirection1Mult, defaultValue=<rgba -1.7 2.4 0 1>}
  n36 = AppendVector(A=@n35.R, B=@n35.R)
  n37 = TextureSample(UVs=Panner(Coordinate=Subtract(A=@n34, B=@n29), Time=@n27, Speed=Multiply(A=@n31, B=@n36)) {speedX=1, speedY=1}) {texture=&T_ZeroPoint_Line_Stars}
  n38 = Multiply(A=@n37.RGB, B=VectorParameter() {group=Color, parameterName=SecondaryTint, defaultValue=<rgba 0.3664 0.0328 0.7 1>}.RGB) {constB=0.5}
  n39 = VectorParameter() {group=Color, parameterName=BackgroundColor1, defaultValue=<rgba 0.25 0.0199 0.027 1>}
  n40 = VectorParameter() {group=Color, parameterName=BackgroundColor2, defaultValue=<rgba 0.0584 0.0394 0.2 0>}
  n41 = NamedRerouteUsage()
  n42 = ScalarParameter() {defaultValue=0.5, group=End Fade, parameterName=EndFadeRadius_X}
  n43 = Multiply(A=SphereMask(A=ComponentMask(None=@n4) {r=true}, B=@n42, Radius=@n42, Hardness=ScalarParameter() {group=End Fade, parameterName=EndFadeHardness_X}) {hardnessPercent=0})
  n44 = MaterialFunctionCall(In=@n43, Contrast=Constant() {r=0.25}) {materialFunction=&CheapContrast}
  n45 = Add(A=Add(A=Add(A=@n33, B=@n38), B=@n28.RGB), B=Multiply(A=LinearInterpolate(A=@n39.RGB, B=@n40.RGB, Alpha=@n41), B=@n44.Result))
  n46 = ParticleColor()
  MaterialFunctionCall(Emissive=Multiply(A=Multiply(A=@n45), B=@n46.RGB)) {materialFunction=&MF_FX_EmissiveAdjustments}
  Multiply(
   A=MaterialFunctionCall() {materialFunction=/Game/Athena/Environments/Nexus/Pickles/Materials/MF_Pickle_Tendril_Colors}.Result
   B=VectorParameter() {group=Color, parameterName=EmissiveMult, defaultValue=<rgba 0.7893 0.75 1 0>}.RGB
  )
  n47 = Add(A=@n32.B, B=Add(A=@n37.B, B=@n28.B))
  n48 = Multiply(A=Multiply(A=Add(A=@n47, B=@n44.Result) {constB=0.5}, B=@n43), B=@n6)
  n49 = Multiply(
   A=MaterialFunctionCall(Distance=ScalarParameter() {defaultValue=4000, group=Opacity, parameterName=NearFadeDIst}) {materialFunction=&Near_Camera_Fade}.Result
   B=DepthFade(FadeDistance=ScalarParameter() {defaultValue=400, group=Opacity, parameterName=DepthFade})
  )
  Multiply(A=@n46.A, B=Multiply(A=Multiply(A=@n48, B=@n49), B=ScalarParameter() {defaultValue=0.15, group=Opacity, parameterName=OpacityMult}))
  NamedRerouteDeclaration(None=@n18.OpenAmount) {name=OpenAmount, nodeColor=<rgba 0.6875 0 1 1>}
  n50 = Add(A=VertexColor().G, B=@n10)
  NamedRerouteDeclaration(None=ConstantBiasScale(None=Sine(None=@n50))) {name=VertexColorSine, nodeColor=<rgba 0 0.3437 1 1>}
 ]
}
