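# Niagara system for the DragonCart shard charge-up effect, driven by the User.ChargeUp float.
# A burst-spawned core glow with a light shrinks, brightens, pulses harder and shifts from purple to orange as charge rises.
# Sparkles spawn on a sphere and curve into the center, becoming denser, shorter-lived and straighter with charge, with speed-based stretching and size scaled by LOD distance.
# The PlasmaWave emitter has no renderers.
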NiagaraSystem /DragonCartNarrative/Assets/VFX/ShardPowered/NS_DragonCart_Shard_ChargeUp
dependencies {
 usedDataInterfaces [&NiagaraDataInterfaceCurve, &NiagaraDataInterfaceColorCurve]
 usedDynamicInputs []
 usedModules [&SystemState]
}
emitters {
 Minimal {
  emitterUpdateScript [
   &EmitterState(
    Inactive Response=::NewEnumerator1
    Life Cycle Mode=::NewEnumerator1
    Loop Behavior=::NewEnumerator1
    Loop Duration Mode=::NewEnumerator1
    UseLoopDelay=false
   )
   &SpawnBurst_Instantaneous(Use Spawn Probability=false, UseLoopCountLimit=false)
  ]
  particleSpawnScript [&InitializeParticle(
    Color=<rgba 16 16 16 1>
    Mass=700
    Mass Mode=::NewEnumerator1
    Sprite Size Mode=::NewEnumerator3
    Uniform Sprite Size=3000
    UsePositionOffset=false
    Write Material Random=true
   )]
  particleUpdateScript [
   &ParticleState(Kill Particles When Lifetime Has Elapsed=false)
   &ScaleSpriteSize(
    Uniform Curve Index=$User.ChargeUp
    Uniform Curve Scale=&Waveform(
     "Waveform [1]"=::NewEnumerator0
     "[1] Phase Scale"=6
     "[1] Frequency"=1
     "[1] Amplitude Min/Max"=&MakeVector2D(X=&Lerp_Float(A=1, B=0.9, Alpha=$User.ChargeUp), Y=1)
    )
    Uniform Curve Sprite Scale=DataInterface() {propertyValues {Curve {
       defaultValue=0.05
       keys [
        {arriveTangent=-0.5778, interpMode=RCIM_Cubic, leaveTangent=-0.5778, tangentMode=RCTM_User, time=0, value=1}
        {arriveTangent=-0.5683, interpMode=RCIM_Cubic, leaveTangent=-0.5683, tangentMode=RCTM_User, time=1, value=0.4}
       ]
      }}}
   )
   &ScaleColor(
    Scale Alpha=&FloatFromCurve(
     FloatCurve=DataInterface() {propertyValues {Curve {
        defaultValue=0
        keys [
         {arriveTangent=1.598, interpMode=RCIM_Cubic, leaveTangent=1.598, tangentMode=RCTM_User, time=0, value=0}
         {arriveTangent=2.4873, tangentMode=RCTM_User, time=1, value=1}
        ]
       }}}
     CurveIndex=$User.ChargeUp
    )
    Scale RGB=&MakeVectorFromLinearColorRGB(Color=&ColorFromCurve(
      ColorCurve=DataInterface() {propertyValues {
        AlphaCurve {keys [<at 0 1 RCIM_Cubic>]}
        BlueCurve {keys [<at 0.3196 0.6128 RCIM_Cubic>,<at 1 0.0026 RCIM_Cubic>]}
        GreenCurve {keys [<at 0.3196 0.2186 RCIM_Cubic>,<at 1 0.7194 RCIM_Cubic>]}
        RedCurve {keys [<at 0.3196 0.7 RCIM_Cubic>,<at 1 5 RCIM_Cubic>]}
       }}
      CurveIndex=$User.ChargeUp
     ))
    ScaleA=true
    ScaleRGB=true
   )
   /SetVariables(Particles.LightColor=&ColorFromCurve(
     ColorCurve=DataInterface() {propertyValues {
       AlphaCurve {keys [<at 0 0 RCIM_Cubic>,<at 0.7149 1 RCIM_Cubic>,<at 1 0 RCIM_Cubic>]}
       BlueCurve {keys [<at 0.2582 10 RCIM_Cubic>,<at 0.9862 0.8481 RCIM_Cubic>]}
       GreenCurve {keys [<at 0.2582 0.7088 RCIM_Cubic>,<at 0.9862 11.2211 RCIM_Cubic>]}
       RedCurve {keys [<at 0.2582 7.1573 RCIM_Cubic>,<at 0.9862 25 RCIM_Cubic>]}
      }}
     CurveIndex=$User.ChargeUp
    )) {isSetParameters=true}
   &CameraOffset(Camera Offset Amount=/Niagara/DynamicInputs/Math/CalculateParticleRadius(
     Radius Calculation Type:ENiagaraCollisionRadiusOptions=::NewEnumerator0
     Particle Radius:NiagaraFloat=1
     Mesh Dimensions:Vector3f=<1 1 1>
     Method for Calculating Particle Radius:ENiagaraCalculateRadiusOptions=::NewEnumerator2
    ))
   &DynamicMaterialParameters(
    Index 0 Param 1=1
    Index 0 Param 2=&Lerp_Float(A=0, B=1.2, Alpha=$User.ChargeUp)
    Module.Index0Float0Write=true
    Module.Index0Float1Write=true
    Module.Index0Float2Write=false
    Module.Index0Float3Write=false
   )
  ]
  rendererClasses [&NiagaraSpriteRendererProperties, &NiagaraLightRendererProperties]
  renderers [
   {rendererClass=&NiagaraSpriteRendererProperties}
   {rendererClass=&NiagaraLightRendererProperties, rendererIndex=1}
  ]
 }
 PlasmaWave {rendererClasses [], renderers []}
 Sparkles {
  emitterUpdateScript [
   &EmitterState(Life Cycle Mode=::NewEnumerator1, Loop Behavior=::NewEnumerator1, Loop Duration Mode=::NewEnumerator1, UseLoopDelay=false)
   &SpawnRate(SpawnRate=&Lerp_Float(A=30, B=160, Alpha=$User.ChargeUp), Use Spawn Probability=false)
  ]
  particleSpawnScript [
   &InitializeParticle(
    Color Maximum=<rgba 1.1912 0.2621 5 1>
    Color Minimum=<rgba 3 0.1122 2.1565 1>
    Color Mode=::NewEnumerator2
    Lifetime=&Lerp_Float(A=3, B=1, Alpha=$User.ChargeUp)
    Sprite Size Mode=::NewEnumerator1
    Uniform Sprite Size Max=120
    Uniform Sprite Size Min=90
    UsePositionOffset=false
    Write Material Random=true
   )
   &ShapeLocation(
    Hemisphere Distribution=<1 0.7>
    MaskSpawnGroup=false
    Override Seed=false
    Sphere Radius=&Lerp_Float(A=1500, B=2400, Alpha=$User.ChargeUp)
    Sphere Surface Distribution=1
    UseOwnerScale=true
   )
  ]
  particleUpdateScript [
   &ParticleState()
   /SetVariables(Particles.Position=&LerpPosition(
     A=&RotateVectorByQuaternion(
      Vector To Rotate=Linked() {linkedVariable {name=Particles.Initial.Position, type=NiagaraPosition}}
      Quaternion=&MakeQuaternion(
       Quaternion From=0
       Axis=&RandomVector(Vector Scale=1)
       Angle=&FloatFromCurve(
        FloatCurve=DataInterface() {propertyValues {Curve {keys [{arriveTangent=-1.7952, interpMode=RCIM_Cubic, leaveTangent=-1.7952, tangentMode=RCTM_User, time=0, value=1}, <at 1 0 RCIM_Cubic>]}}}
        Scale Curve=&Lerp_Float(A=0.15, B=0, Alpha=$User.ChargeUp)
       )
       From=?
       To=?
       Angle Type=::NewEnumerator1
       Normalize Vectors=false
      )
     )
     B=<0 0 0>
     Alpha=&FloatFromCurve(FloatCurve=DataInterface() {propertyValues {Curve {
         defaultValue=1
         keys [
          {arriveTangent=0.1583, interpMode=RCIM_Cubic, leaveTangent=0.1583, tangentMode=RCTM_User, time=0, value=0}
          {arriveTangent=2.1135, interpMode=RCIM_Cubic, leaveTangent=2.1135, tangentMode=RCTM_User, time=1, value=1}
         ]
        }}})
    )) {isSetParameters=true}
   &CalculateAccurateVelocity()
   &ScaleSpriteSizeBySpeed(
    Max Scale Factor=<0.9 8>
    Sample Scale Factor By Curve=true
    Scale Factor Curve=DataInterface() {propertyValues {Curve {keys [<at 0 0 RCIM_Cubic>, {arriveTangent=1.7805, tangentMode=RCTM_User, time=1, value=1}]}}}
    Velocity Threshold=4000
   )
   &ScaleSpriteSize(
    Uniform Curve Scale=&FloatFromCurve(FloatCurve=DataInterface() {propertyValues {Curve {keys [<at 8000 0.25>,<at 30000 1>]}}}, CurveIndex=$Engine.Owner.LODDistance)
    Uniform Curve Sprite Scale=DataInterface() {propertyValues {Curve {keys [
        {arriveTangent=68.9421, interpMode=RCIM_Cubic, time=0, value=0}
        {arriveTangentWeight=0.05, interpMode=RCIM_Cubic, leaveTangentWeight=0.2167, tangentMode=RCTM_User, time=0.25, value=1}
        {
         arriveTangent=2.0866
         arriveTangentWeight=0.3345
         interpMode=RCIM_Cubic
         leaveTangent=2.0866
         leaveTangentWeight=0.1029
         tangentMode=RCTM_User
         time=1
         value=1.5
        }
       ]}}}
   )
   &DynamicMaterialParameters(
    Index 0 Param 1=2
    Index 0 Param 3=&FloatFromCurve(FloatCurve=DataInterface() {propertyValues {Curve {defaultValue=1, keys [<at 0 0 RCIM_Cubic>, {tangentMode=RCTM_User, time=1, value=3}]}}})
    Index 0 Param 4=&FloatFromCurve(FloatCurve=DataInterface() {propertyValues {Curve {keys [<at 0 1 RCIM_Cubic>,<at 0.25 0.2>]}}})
    Module.Index0Float0Write=true
    Module.Index0Float1Write=false
    Module.Index0Float2Write=true
    Module.Index0Float3Write=true
   )
   &ScaleColor(
    Scale Alpha=&FloatFromCurve(FloatCurve=DataInterface() {propertyValues {Curve {keys [<at 0.5 1 RCIM_Cubic>, {arriveTangent=-4.8071, tangentMode=RCTM_User, time=1, value=0}]}}})
    Scale RGB=&VectorFromFloat(Value=&FloatFromCurve(FloatCurve=DataInterface() {propertyValues {Curve {keys [{arriveTangent=-13.9762, interpMode=RCIM_Cubic, leaveTangent=-13.9762, tangentMode=RCTM_User, time=0, value=5}, <at 0.5 1>]}}}))
    ScaleA=true
    ScaleRGB=true
   )
  ]
  rendererClasses [&NiagaraSpriteRendererProperties]
  renderers [{rendererClass=&NiagaraSpriteRendererProperties}]
 }
}
summary {
 systemName=NS_DragonCart_Shard_ChargeUp
 emitters [
  {emitterName=PlasmaWave, rendererClasses []}
  {emitterName=Sparkles, rendererClasses [&NiagaraSpriteRendererProperties]}
  {emitterName=Minimal, rendererClasses [&NiagaraSpriteRendererProperties, &NiagaraLightRendererProperties]}
 ]
 userVariables [User.ChargeUp {type=NiagaraFloat, defaultValue=1, description=""}]
}
