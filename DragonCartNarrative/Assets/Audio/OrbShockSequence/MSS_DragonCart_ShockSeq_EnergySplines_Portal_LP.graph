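# Mono MetaSound source for the looping energy-spline portal sound in the DragonCart orb shock sequence.
# On play, a random time between zero and the wave's duration sets the Wave Player's start position, so each instance begins at a different point in the loop.
# It is spatialized through the shard splines attenuation asset and routed to the Fort_Events sound class.
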MetaSoundSource /DragonCartNarrative/Assets/Audio/OrbShockSequence/MSS_DragonCart_ShockSeq_EnergySplines_Portal_LP
inputs [
 Sound {type=WaveAsset}
 UE.Source.OnPlay {type=Trigger}
]
interfaces [UE.OutputFormat.Mono:1.0, UE.Source:1.1]
outputs ["UE.OutputFormat.Mono.Audio:0" {type=Audio}]
pages [Default {
  input_defaults {Sound=/DragonCartNarrative/Assets/Audio/OrbShockSequence/Wav/DragonCart_ShockSeq_Shard_EnergySpline_Portal_LP, UE.Source.OnPlay=null}
  nodes [
   GetWaveDuration_0 = "UE.Get Wave Duration: "(Wave=^Sound)
   n1 = UE.RandomTime(Next=^UE.Source.OnPlay, Max=@GetWaveDuration_0.Duration)
   WavePlayer_0 = UE.Wave Player:Mono(Wave Asset=^Sound, Play=@n1.On Next, Start Time=@n1.Value) {also {defaults {Loop="<MetaSoundBuilderResult.SUCCEEDED: 0>"}, version [1,1]}}
  ]
  links [{dst="@UE.OutputFormat.Mono.Audio:0", src=WavePlayer_0.Out Mono}]
 }]
properties {
 assetImportData=&self:AssetImportData
 attenuationSettings=&/ATT_DragonCart_ShardSplines
 bLooping=true
 duration=10000
 maxDistance=80000
 sampleRate=48000
 soundClassObject=&Fort_Events
 volume=0.6
}
