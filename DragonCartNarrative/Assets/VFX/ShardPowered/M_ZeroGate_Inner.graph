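# Unlit translucent two-sided material for Niagara mesh particles, drawing the inside of a Zero Gate portal as a window onto a DragonCart shard image.
# Screen UVs anchor to a projected offset point that shifts with camera height and scales with distance, faking interior depth.
# Panning noise, caustics and magnetic-wave lines distort the image, add tinted edge glow and drive chromatic splitting, with cheaper quality fallbacks.
# MF_ZerOGate_OpenWPO's open amount, masks and depth/camera fades shape opacity.
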Material /DragonCartNarrative/Assets/VFX/ShardPowered/M_ZeroGate_Inner
graph {
 properties {
  bAutomaticallySetUsageInEditor=false
  bCastRayTracedShadows=false
  bUsedWithNiagaraMeshParticles=true
  blendMode=BLEND_Translucent
  pixelDepthOffsetMode=PDOM_Legacy
  shadingModel=MSM_Unlit
  twoSided=true
  usedShadingModels=Unlit
 }
 nodes [
  ActorPositionWS()
  Constant()
  n1 = ParticleColor()
  n2 = ScreenPosition()
  n3 = Multiply(A=ScalarParameter() {defaultValue=3000, parameterName=DepthOffset}, B=TwoSidedSign())
  n4 = ParticlePositionWS()
  n5 = Subtract(
   A=ComponentMask(None=CameraPositionWS()) {b=true}
   B=ComponentMask(None=NamedRerouteDeclaration(None=@n4) {name=Actor/ParticlePosition, nodeColor=<rgba 0.1719 0 1 1>}) {b=true}
  )
  n6 = ScalarParameter() {defaultValue=2000, parameterName=HeightOffsetMaskSize}
  n7 = LinearInterpolate(
   A=ScalarParameter() {defaultValue=1000, parameterName=HeightOffsetBase}
   B=ScalarParameter() {defaultValue=-2000, parameterName=HeightOffsetTop}
   Alpha=Saturate(None=Divide(A=@n5, B=@n6))
  ) {constB=0}
  n8 = MaterialFunctionCall(World Position=TransformPosition(None=MaterialFunctionCall(X=@n3, Z=@n7) {materialFunction=&MakeFloat3}.Result) {transformType=TRANSFORMPOSSOURCE_World}) {materialFunction=&TransformToClipSpace}
  n9 = Divide(A=ScalarParameter() {defaultValue=1200, parameterName=DistanceScale}, B=Distance(A=NamedRerouteUsage(), B=CameraPositionWS()))
  n10 = ScalarParameter() {defaultValue=0.75, parameterName=MaxDistanceScaleUp}
  n11 = MaterialFunctionCall(Float2=ViewSize()) {materialFunction=&BreakOutFloat2Components}
  n12 = AppendVector(A=Constant() {r=1}, B=Divide(A=@n11.R, B=@n11.R))
  n13 = ScalarParameter() {defaultValue=0.5, parameterName=ScaleMult}
  n14 = MaterialFunctionCall(
   UVs=Add(A=Subtract(A=@n2.ViewportUV, B=@n8.Buffer UVs for Scene Textures)) {constB=0.5}
   Texture Scale=Multiply(A=Clamp(None=@n9, Max=@n10) {clampMode=CMODE_ClampMax, maxDefault=0}, B=Multiply(A=@n12, B=@n13))
  ) {materialFunction=&ScaleUVsByCenter}
  n15 = Multiply(A=@n14.UVs) {constB=3}
  n16 = TextureCoordinate() {uTiling=4, vTiling=2}
  n17 = Time()
  n18 = ScalarParameter() {defaultValue=2, parameterName=PanSpeedDistortion}
  n19 = TextureSample(UVs=Panner(Coordinate=LinearInterpolate(A=@n15, B=@n16) {constAlpha=0.6}, Time=Multiply(A=@n17, B=@n18)) {speedX=0.01, speedY=-0.07}) {texture=/Game/Effects/Fort_Effects/Textures/Tile/T_Smooth_Tile_Chromatic}
  n20 = Constant2Vector() {g=1, r=0.05}
  n21 = VectorParameter() {parameterName=DistortionTile, defaultValue=<rgba 0.3 0.3 2 0>}
  n22 = AppendVector(A=@n21.R, B=@n21.R)
  n23 = Add(A=Multiply(A=@n16, B=Multiply(A=@n20, B=@n22)), B=Constant2Vector() {r=1.75})
  n24 = Multiply(A=@n16, B=@n22)
  n25 = NamedRerouteUsage()
  n26 = Time()
  n27 = ScalarParameter() {defaultValue=0.1, parameterName=PanSpeed}
  n28 = TextureSampleParameter2D(UVs=Panner(Coordinate=LinearInterpolate(A=@n23, B=@n24, Alpha=@n25) {constA=4}, Time=Multiply(A=@n26, B=@n27)) {speedY=-0.2}) {parameterName=DistortionTexture, texture=/Game/Effects/Fort_Effects/Textures/Tile/T_Caustics_Chromatic}
  n29 = SphereMask(
   A=ComponentMask(None=TextureCoordinate()) {r=true}
   B=Constant() {r=0.5}
   Radius=ScalarParameter() {defaultValue=1, parameterName=InteriorLinesRadius}
  ) {attenuationRadius=500, hardnessPercent=0}
  n30 = LinearInterpolate(Alpha=@n29) {constA=1.6, constB=-0.4}
  n31 = Power(Base=@n29, Exp=ScalarParameter() {defaultValue=3, parameterName=InteriorLinesPower})
  n32 = ComponentMask(None=TextureCoordinate()) {g=true}
  n33 = VectorParameter() {parameterName=LinesTile, defaultValue=<rgba 5 3 0 0>}
  n34 = Multiply(A=@n32, B=@n33.R)
  n35 = Panner(Coordinate=Multiply(A=AppendVector(A=@n31, B=@n32), B=AppendVector(A=@n34, B=@n33.G)) {constB=2}) {speedX=-0.1, speedY=-0.17}
  n36 = QualitySwitch(
   Default=Multiply(A=Saturate(None=@n30), B=TextureSample(UVs=@n35) {texture=/Game/Athena/Environments/Nexus/Textures/T_ZeroPoint_MagneticWaves}.RGB) {constB=0}
   Low=Constant()
  )
  n37 = Constant() {r=-1}
  n38 = Add(A=@n19.RGB, B=Add(A=@n28.RGB, B=Desaturation(None=@n36, Fraction=@n37)))
  n39 = TextureCoordinate()
  n40 = ComponentMask(None=@n39) {g=true}
  n41 = Constant() {r=0.5}
  n42 = Multiply(
   A=Power(Base=SphereMask(A=@n40, B=@n41, Radius=@n41) {attenuationRadius=1, hardnessPercent=20})
   B=SphereMask(A=ComponentMask(None=@n39) {r=true}, B=@n41, Radius=@n41) {attenuationRadius=1, hardnessPercent=50}
  )
  n43 = LinearInterpolate(A=@n19.RGB, B=@n28.RGB)
  n44 = Power(
   Base=OneMinus(None=@n42)
   Exp=LinearInterpolate(
    A=ScalarParameter() {defaultValue=1, parameterName=EdgeMaskPower}
    B=ScalarParameter() {parameterName=ExgeMaskPowerMin}
    Alpha=Desaturation(None=@n43)
   ) {constA=3, constB=0.25}
  ) {constExponent=0.5}
  n45 = Multiply(A=Multiply(A=VectorParameter() {parameterName=EdgeColor, defaultValue=<rgba 2 0.8655 1.6353 1>}.RGB, B=@n38), B=Multiply(A=@n44, B=@n44))
  n46 = Multiply(A=@n44, B=Multiply(A=ComponentMask(None=@n36) {g=true, r=true}, B=ScalarParameter() {defaultValue=0.3, parameterName=EdgeRippleDistortion}))
  n47 = Add(
   A=@n14.UVs
   B=Multiply(A=ComponentMask(None=@n19.RGB) {g=true, r=true}, B=ScalarParameter() {defaultValue=0.017, parameterName=MainDistortion})
  ) {constB=0}
  n48 = NamedRerouteDeclaration(None=Add(A=@n47, B=NamedRerouteDeclaration(None=@n46) {name=Ripples, nodeColor=<rgba 0 1 0.8359 1>}) {constB=0}) {name=Distorted UVs, nodeColor=<rgba 0.0078 0 1 1>}
  n49 = TextureCoordinate()
  n50 = Normalize(VectorInput=AppendVector(A=Subtract(A=@n49) {constB=0.5}, B=Constant()))
  n51 = Multiply(A=ScalarParameter() {defaultValue=0.005, parameterName=ChromaticAmount}, B=Add(A=Add(A=@n38, B=@n36)) {constB=1.5})
  n52 = MaterialFunctionCall(Float3=Multiply(A=@n44, B=Multiply(A=@n50, B=@n51) {constB=2})) {desc=Chromatic offset, materialFunction=&BreakOutFloat3Components}
  n53 = TextureSampleParameter2D(UVs=@n48) {parameterName=InteriorTexture, texture=&/T_DragonCart_Shard_04}
  n54 = MaterialFunctionCall(
   R=TextureSampleParameter2D(UVs=Add(A=@n48, B=@n52.R)) {parameterName=InteriorTexture, texture=&T_DragonCart_Shard_04}.R
   G=TextureSampleParameter2D(UVs=Add(A=@n48, B=@n52.G)) {parameterName=InteriorTexture, texture=&T_DragonCart_Shard_04}.G
   B=@n53.B
  ) {materialFunction=&AppendMany}
  n55 = ShadingPathSwitch(Default=@n54.RGB, Forward=@n53.RGB, Mobile=@n53.RGB)
  MaterialFunctionCall(Emissive=Add(A=Multiply(A=@n1.RGB, B=@n45), B=QualitySwitch(Default=@n55, Low=@n53.RGB, Medium=@n54.RGB))) {materialFunction=&MF_FX_EmissiveAdjustments}
  n56 = MaterialFunctionCall!(SphereRadius, SphereCenter) {materialFunction=&/MF_ZerOGate_OpenWPO}
  n57 = Multiply(A=Multiply(A=Power(Base=@n56.OpenAmount) {constExponent=0.6}, B=@n42), B=ScalarParameter() {defaultValue=1, parameterName=OpacityMult})
  n58 = SphereMask(
   A=@n47
   B=Constant() {r=0.5}
   Radius=ScalarParameter() {defaultValue=0.65, parameterName=OpacitySphereRadius}
   Hardness=ScalarParameter() {defaultValue=0.3, parameterName=OpacitySphereHardness}
  ) {hardnessPercent=0.7}
  n59 = MaterialFunctionCall(Distance=ScalarParameter() {defaultValue=800, parameterName=NearCamFade}) {materialFunction=&Near_Camera_Fade}
  Multiply(
   A=DepthFade(Opacity=Multiply(A=@n57, B=@n58), FadeDistance=ScalarParameter() {defaultValue=200, parameterName=DepthFade})
   B=Multiply(A=@n59.Result, B=@n59.Result)
  )
  NamedRerouteDeclaration(None=@n56.OpenAmount) {name=OpenAmount, nodeColor=<rgba 1 0 0.9656 1>}
  NamedRerouteDeclaration(None=@n56.PortalWPO) {name=Port OpenWPO, nodeColor=<rgba 0.6172 1 0 1>}
  NamedRerouteUsage()
 ]
}
