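# Niagara system for the DragonCart shard gate: a mesh portal with light, looping sound and swirling sprite motes.
# Gate derives a smoothed OpenAmount from camera distance to a point ahead of it, driving light, upper-spline color and scale, and audio.
# That point and open radius also feed material parameters.
# Minimal spawns cylinder motes shaped by vortex, drag and outward acceleration; both emitters cull by distance.
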NiagaraSystem /DragonCartNarrative/Assets/VFX/ShardPowered/NS_DragonCart_Shard_Gate
dependencies {
 usedDataInterfaces [&NiagaraDataInterfaceAudioPlayer, &NiagaraDataInterfaceCurve, &NiagaraDataInterfaceCamera]
 usedDynamicInputs []
 usedModules [&SystemState]
}
emitters {
 Gate {
  emitterSpawnScript [&InitializeAudioPlayer(Audio Player=DataInterface() {propertyValues {
      Attenuation=null
      Concurrency=null
      MaxPlaysPerTick=10
      SoundToPlay=/DragonCartNarrative/Assets/Audio/OrbShockSequence/MSS_DragonCart_ShockSeq_EnergySplines_Portal_LP
      bAllowLoopingOneShotSounds=false
      bLimitPlaysPerTick=true
      bOnlyActiveDuringGameplay=false
      bStopWhenComponentIsDestroyed=true
      ConfigurationUserParameter {parameter {name=None, typeDefHandle {registeredTypeIndex=1}, varData []}}
      ParameterNames []
     }})]
  emitterUpdateScript [
   &EmitterState(
    Enable Distance Culling=true
    Inactive Response=::NewEnumerator1
    Life Cycle Mode=::NewEnumerator1
    Loop Behavior=::NewEnumerator1
    Loop Duration Mode=::NewEnumerator1
    Max Distance Response=::SleepAndClearParticles
    MaxDistance=16000
    Module.bUseMaxDistance=true
    Module.bUseMinDistance=false
    Reset Age On Awaken=true
    Scalability Mode=::NewEnumerator1
    UseLoopDelay=false
   )
   &SpawnBurst_Instantaneous(Use Spawn Probability=false, UseLoopCountLimit=false)
  ]
  particleSpawnScript [
   &PartitionParticles(Number Of Partitions=3)
   &InitializeParticle(Color=<rgba 1.3429 0.7788 1.5 1>, Lifetime=5, UsePositionOffset=false, Write Material Random=true)
   &InitialMeshOrientation(Mesh Orientation Mode=::NewEnumerator2, Orientation Coordinate Space=::Local, Rotation=<0 -0.02 0>, Use Rotation Vector=true)
   /SetVariables(Particles.OpenRadius=700) {isSetParameters=true}
   /SetVariables(Particles.OffsetPosition:NiagaraPosition=&AddVectorToPosition(
     Position=$Particles.Position
     Vector=&Multiply_VectorByFloat(
      Vector=&ReturnMeshOrientationAxis(Mesh Orientation Axis=::XAxis, Mesh Orientation Quat=$Particles.MeshOrientation)
      Float=1500
     )
    )) {isSetParameters=true}
   &DynamicMaterialParameters(
    Index 0 Param 4=$Particles.OpenRadius
    Index 0 Vector 1:NiagaraPosition=$Particles.OffsetPosition
    Index 3 Param 1=&RandomRangeFloat(Minimum=0, Maximum=1)
    Index 3 Param 2=2
    Index 3 Param 4=1
    Module.Index0Float3Write=true
    Module.Index0VectorWrite=true
    Module.Index1Float0Write=false
    Module.Index1Float1Write=false
    Module.Index1Float2Write=false
    Module.Index1Float3Write=false
    Module.Index3Float0Write=true
    Module.Index3Float1Write=true
    Module.Index3Float2Write=true
    Module.Index3Float3Write=true
    Write Parameter Index 0=::NewEnumerator3
    Write Parameter Index 1=::NewEnumerator1
    Write Parameter Index 3=::NewEnumerator1
   )
   /SetVariables(Particles.PreviewScale:Vector3f=&VectorFromFloat(Value=&Divide_Float(A=$Particles.OpenRadius, B=50))) {isSetParameters=true}
  ]
  particleUpdateScript [
   &ParticleState(Kill Particles When Lifetime Has Elapsed=false)
   &InterpolateOverTime(
    Rate Of Change=7
    Target Float Value:NiagaraFloat=&FloatFromCurve(
     FloatCurve=DataInterface() {propertyValues {Curve {defaultValue=1, keys [<at 0.2 1 RCIM_Cubic>, {arriveTangentWeight=0.3, interpMode=RCIM_Cubic, tangentMode=RCTM_User, time=1, value=0}]}}}
     CurveIndex=&Divide_Float(
      A=&DistanceBetweenPositions(
       A=$Particles.OffsetPosition
       B=&ReturnCameraPosition(
        Coordinate Space:ENiagaraCoordinateSpace=::World
        Camera Query:NiagaraDataInterfaceCamera=DataInterface() {propertyValues {bRequireCurrentFrameData=false}}
       )
      )
      B=$Particles.OpenRadius
     )
    )
   )
   /SetVariables(Particles.OpenAmount:NiagaraFloat=$Particles.InterpolateOverTime.InterpolateOverTime_Float.Moving Average) {isSetParameters=true}
   /SetVariables(Particles.ShouldPlayAudio:NiagaraBool=&SetBoolByFloatComparison(A=$Particles.OpenAmount, B=0)) {isSetParameters=true}
   &ScaleColor(
    Scale Alpha=&FloatFromCurve(
     FloatCurve=DataInterface() {propertyValues {Curve {defaultValue=1, keys [<at 0 0 RCIM_Cubic>, {tangentMode=RCTM_User, time=1, value=1}]}}}
     Scale Curve=&FloatFromCurve(
      FloatCurve=DataInterface() {propertyValues {Curve {defaultValue=0, keys [<at 14000 1 RCIM_Cubic>,<at 16000 0 RCIM_Cubic>]}}}
      CurveIndex=$Engine.Owner.LODDistance
     )
    )
    ScaleA=true
    ScaleRGB=false
   )
   /SetVariables(
    Particles.LightColor=&MultiplyLinearColor(
     Linear Color=$Particles.Color
     Scale Factor=&MakeVector4_FromLinearColor(LinearColor=&LerpLinearColors(StartColor=<rgba 5 0.2865 2.6303 1>, EndColor=<rgba 6.54 1.08 12 1>, LerpFactor=$Particles.OpenAmount))
    )
    Particles.LightPosition:NiagaraPosition=&AddVectorToPosition(
     Position=&LerpPosition(A=$Particles.Position, B=$Particles.OffsetPosition, Alpha=&Multiply_Float(A=$Particles.OpenAmount, B=0.2))
     Vector=<0 0 20>
    )
    Particles.LightRadius=&Lerp_Float(A=500, B=1000, Alpha=$Particles.OpenAmount)
   ) {isSetParameters=true}
   /SetVariables(
    Particles.UpperSplineColor:LinearColor=&MultiplyLinearColor(
     Linear Color=$Particles.Color
     Scale Factor=&MakeVector4(X=1, Y=1, Z=1, W=&Lerp_Float(A=0.4, B=0.5, Alpha=$Particles.OpenAmount))
    )
    Particles.UpperSplineScale:Vector3f=&Lerp_Vector(A=<1 0.9 0.9>, B=<1 1.5 1.5>, Alpha=$Particles.OpenAmount)
   ) {isSetParameters=true}
   /Niagara/Modules/Audio/PlayPersistentAudio(Enable Fade In / Out=true, Fade Out Time=2, Play Audio:NiagaraBool=$Particles.ShouldPlayAudio)
   /Niagara/Modules/Audio/UpdatePersistentAudio(Volume:NiagaraFloat=&FloatFromCurve(FloatCurve=DataInterface() {propertyValues {Curve {keys [<at 0 0.05>,<at 0.1 0.3>,<at 1 1>]}}}, CurveIndex=$Particles.OpenAmount))
  ]
  rendererClasses [&NiagaraMeshRendererProperties, &NiagaraLightRendererProperties]
  renderers [
   {rendererClass=&NiagaraMeshRendererProperties, rendererIndex=0}
   {rendererClass=&NiagaraMeshRendererProperties, rendererIndex=1}
   {rendererClass=&NiagaraLightRendererProperties, rendererIndex=2}
   {rendererClass=&NiagaraMeshRendererProperties, rendererIndex=3}
  ]
 }
 Minimal {
  emitterUpdateScript [
   &EmitterState(Enable Distance Culling=true, MaxDistance=8000, Module.bUseMaxDistance=true, Module.bUseMinDistance=false, Scalability Mode=::NewEnumerator1)
   &SpawnRate(SpawnRate=7, Use Spawn Probability=false)
  ]
  particleSpawnScript [
   &InitializeParticle(
    Color Maximum=<rgba 3 0.2145 2.1212 1>
    Color Minimum=<rgba 0.5015 0.0986 1 1>
    Color Mode=::NewEnumerator2
    Lifetime Max=10
    Lifetime Min=4
    Lifetime Mode=::NewEnumerator1
    Sprite Rotation Mode=::NewEnumerator2
    Sprite Size Mode=::NewEnumerator1
    Uniform Sprite Size Max=20
    Uniform Sprite Size Min=10
    UsePositionOffset=false
    Write Material Random=true
   )
   &ShapeLocation(
    Cylinder Height=800
    Cylinder Height Midpoint=0.15
    Cylinder Radius=70
    MaskSpawnGroup=false
    Override Seed=false
    Shape Primitive=::NewEnumerator1
    Surface Only=false
    UseOwnerScale=true
   )
   &AddVelocity(
    Module.Use Falloff Distance=false
    Origin Offset=<0 0 1000>
    Override Seed=false
    Velocity Mode=::NewEnumerator1
    Velocity Speed=&RandomRangeFloat(Minimum=50, Maximum=200)
   )
   /SetVariables(Particles.Velocity=&Multiply_Vector(A=$Particles.Velocity, B=<1 1 0.25>)) {isSetParameters=true}
   &DynamicMaterialParameters(
    Index 0 Param 1=&RandomRangeFloat(Minimum=0.1, Maximum=0.5)
    Index 0 Param 2=&RandomRangeFloat(Minimum=0.4, Maximum=0.8)
    Module.Index0Float0Write=true
    Module.Index0Float1Write=true
    Module.Index0Float2Write=false
    Module.Index0Float3Write=false
   )
   &CameraOffset(Camera Offset Amount=100)
  ]
  particleUpdateScript [
   &ParticleState()
   &ScaleSpriteSize(Uniform Curve Sprite Scale=DataInterface() {propertyValues {Curve {
       defaultValue=0
       keys [
        {interpMode=RCIM_Cubic, tangentMode=RCTM_User, time=0, value=0}
        {arriveTangentWeight=0.1, interpMode=RCIM_Cubic, leaveTangentWeight=0.1, tangentMode=RCTM_User, time=0.3, value=1}
        {arriveTangentWeight=0.1744, interpMode=RCIM_Cubic, leaveTangentWeight=0.2325, tangentMode=RCTM_User, time=0.6, value=1}
        {tangentMode=RCTM_User, time=1, value=0}
       ]
      }}})
   &VortexForce(
    Origin Pull Amount:NiagaraFloat=&FloatFromCurve(
     FloatCurve=DataInterface() {propertyValues {Curve {
        defaultValue=1
        keys [
         {arriveTangent=6.0574, interpMode=RCIM_Cubic, time=0.5, value=0}
         {arriveTangentWeight=0.1864, interpMode=RCIM_Cubic, leaveTangentWeight=0.1856, tangentMode=RCTM_User, time=1.5, value=1}
         <at 3 0.5 RCIM_Cubic>
        ]
       }}}
     CurveIndex=$Particles.Age
     Scale Curve=350
    )
    UseOriginOffset=true
    Vortex Axis=<0 -1 0>
    Vortex Origin Offset:Vector3f=<500 0 100>
   )
   &Drag(
    Drag=&FloatFromCurve(
     FloatCurve=DataInterface() {propertyValues {Curve {
        defaultValue=1
        keys [
         {arriveTangent=1.4702, interpMode=RCIM_Cubic, leaveTangent=1.4702, tangentMode=RCTM_User, time=0, value=0}
         {tangentMode=RCTM_User, time=2, value=1}
        ]
       }}}
     CurveIndex=$Particles.Age
     Scale Curve=2
    )
    Use Linear Drag=true
    Use Rotational Drag=false
   )
   &AccelerationForce(Acceleration=&Multiply_VectorByFloat(
     Vector=&Multiply_Vector(A=$Particles.ShapeLocation.ShapeNormal, B=<1 1 0>)
     Float=&FloatFromCurve(
      FloatCurve=DataInterface() {propertyValues {Curve {defaultValue=1, keys [<at 0.5 0 RCIM_Cubic>, {tangentMode=RCTM_User, time=4, value=1}]}}}
      CurveIndex=$Particles.Age
      Scale Curve=100
     )
    ))
   &SolveForcesAndVelocity(Clamp Velocity=false, Limit Acceleration=false)
  ]
  rendererClasses [&NiagaraSpriteRendererProperties]
  renderers [{rendererClass=&NiagaraSpriteRendererProperties}]
 }
}
summary {
 systemName=NS_DragonCart_Shard_Gate
 emitters [
  {emitterName=Gate, rendererClasses [&NiagaraMeshRendererProperties, &NiagaraLightRendererProperties]}
  {emitterName=Minimal, rendererClasses [&NiagaraSpriteRendererProperties]}
 ]
 userVariables [User.Material {
   type=MaterialInterface
   description=""
   defaultValue {
    struct=NiagaraExt_VariableValue_Object
    value {object=&/M_ZeroGate_Inner, objectClass=/Script/Engine.MaterialInterface}
   }
  }]
}
