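# Niagara effect for the DragonCart shard heart: glowing tendrils along three user-supplied splines, plus slowly expanding, fading meshes emitted around the origin.
# The tendril emitter burst-spawns one ribbon group per spline, then rotates three position and color sets around the spline direction for three ribbon renderers, forming braided strands from shared particles.
# Waveforms pulse width and alpha.
# An alternative nine-group tendril emitter and GPU spline sparkles are disabled.
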NiagaraSystem /DragonCartNarrative/Assets/VFX/ShardPowered/NS_DragonCart_Shard_Heart
dependencies {
 usedDataInterfaces [
  &NiagaraDataInterfaceArrayFloat
  &NiagaraDataInterfaceCurve
  &NiagaraDataInterfaceArrayColor
  &NiagaraDataInterfaceVectorCurve
  /Script/Niagara.NiagaraDataInterfaceSpline
 ]
 usedDynamicInputs [&ModuloFloat]
 usedModules [
  &self:SystemSpawnScript.SystemScriptSource.SystemScriptGraph.NiagaraNodeAssignment_0.SetVariables_#1
  &SystemState
 ]
}
emitters {
 Sparkles {
  bEnabled=false
  simTarget=GPUComputeSim
  emitterUpdateScript [
   &EmitterState()
   &SpawnRate(SpawnRate=200, Use Spawn Probability=false)
  ]
  particleSpawnScript [
   &PartitionParticles(Number Of Partitions=3)
   &InitializeParticle(
    Color Maximum=<rgba 2 0.1025 1.1309 1>
    Color Minimum=<rgba 0.3906 0.0496 1 0.5>
    Color Mode=::NewEnumerator2
    Lifetime Max=4
    Lifetime Min=2
    Lifetime Mode=::NewEnumerator1
    Sprite Size Mode=::NewEnumerator1
    Uniform Sprite Size Max=20
    Uniform Sprite Size Min=10
    UsePositionOffset=false
    Write Material Random=true
   )
   &self:SetPositionOnSpline(
    Spline1:NiagaraDataInterfaceSpline=$System.Spline1
    Spline2:NiagaraDataInterfaceSpline=$System.Spline2
    Spline3:NiagaraDataInterfaceSpline=$System.Spline3
    SplineIndex:NiagaraInt32=$Particles.PartitionParticles.Partition
    float:NiagaraFloat=$Particles.MaterialRandom
   ) {module=SetPositionOnSpline}
   &ShapeLocation(MaskSpawnGroup=false, Override Seed=false, Sphere Radius=200, UseOwnerScale=true)
   &AddVelocity(
    Velocity=$Output.SetPositionOnSpline.SampledDirection
    Velocity Speed Scale=&Multiply_Float(
     A=&RandomRangeFloat(Minimum=200, Maximum=400)
     B=&FloatFromCurve(
      FloatCurve=DataInterface() {propertyValues {Curve {keys [<at 0.8 1 RCIM_Cubic>, {arriveTangent=-5.6776, interpMode=RCIM_Cubic, leaveTangent=-5.6776, tangentMode=RCTM_User, time=1, value=0.1}]}}}
      CurveIndex=$Particles.MaterialRandom
     )
    )
   )
  ]
  particleUpdateScript [
   &ParticleState()
   &VectorNoiseForce(
    Force Amount=400
    Module.Use Random Noise Time Interval=true
    Random Noise Time Interval:NiagaraFloat=&RandomRangeFloat(Minimum=0.4, Maximum=0.5)
   )
   &SolveForcesAndVelocity(Clamp Velocity=false, Limit Acceleration=false)
   &ScaleColor(
    Scale Alpha=&FloatFromCurve(FloatCurve=DataInterface() {propertyValues {Curve {
        defaultValue=1
        keys [
         {arriveTangent=4.5506, interpMode=RCIM_Cubic, leaveTangent=4.5506, tangentMode=RCTM_User, time=0, value=0}
         {interpMode=RCIM_Cubic, tangentMode=RCTM_User, time=0.5, value=1}
         {arriveTangent=-3.5267, tangentMode=RCTM_User, time=1, value=0}
        ]
       }}})
    ScaleA=true
    ScaleRGB=false
   )
  ]
  rendererClasses [&NiagaraSpriteRendererProperties]
  renderers [{rendererClass=&NiagaraSpriteRendererProperties}]
 }
 Splines {
  emitterUpdateScript [
   &EmitterState(
    Enable Distance Culling=true
    MaxDistance=60000
    Module.bUseMaxDistance=true
    Module.bUseMinDistance=false
    Scalability Mode=::NewEnumerator1
   )
   &SpawnRate(SpawnRate=1.5, Use Spawn Probability=false)
  ]
  particleSpawnScript [
   &InitializeParticle(
    Color Maximum=<rgba 2 1.2272 1.3616 1>
    Color Minimum=<rgba 1 1 1 1>
    Color Mode=::NewEnumerator2
    Lifetime Max=8
    Lifetime Min=6
    Lifetime Mode=::NewEnumerator1
    Mesh Scale Max=<40 70 70>
    Mesh Scale Min=<20 30 30>
    Mesh Scale Mode=::NewEnumerator2
    UsePositionOffset=false
    Write Material Random=true
   )
   &ShapeLocation(
    Hemisphere Distribution=<1 0.42>
    MaskSpawnGroup=false
    Override Seed=false
    Sphere Radius=50
    Sphere Surface Distribution=1
    UseOwnerScale=true
   )
   &InitialMeshOrientation(
    Mesh Orientation Mode=::NewEnumerator2
    Rotation=&RandomRangeVector(Minimum=<0 0 0>, Maximum=<1 0 0>)
    Use Rotation Vector=true
   )
   &AddVelocity(Module.Use Falloff Distance=false, Override Seed=false, Velocity Mode=::NewEnumerator1, Velocity Speed=100)
  ]
  particleUpdateScript [
   &ParticleState()
   &ScaleMeshSize(Scale Factor=&VectorFromCurve(VectorCurve=DataInterface() {propertyValues {
       XCurve {
        defaultValue=1
        keys [
         {arriveTangent=0.3265, interpMode=RCIM_Cubic, leaveTangent=0.3265, tangentMode=RCTM_User, time=0, value=0.2}
         {arriveTangent=0.4285, interpMode=RCIM_Cubic, leaveTangent=0.4285, tangentMode=RCTM_User, time=1, value=1}
        ]
       }
       YCurve {
        defaultValue=1
        keys [
         {arriveTangent=1.3531, interpMode=RCIM_Cubic, leaveTangent=1.3531, tangentMode=RCTM_User, time=0, value=0.25}
         {arriveTangent=0.3205, interpMode=RCIM_Cubic, leaveTangent=0.3205, tangentMode=RCTM_User, time=1, value=1}
        ]
       }
       ZCurve {
        defaultValue=1
        keys [
         {arriveTangent=1.3531, interpMode=RCIM_Cubic, leaveTangent=1.3531, tangentMode=RCTM_User, time=0, value=0.25}
         {arriveTangent=0.3205, interpMode=RCIM_Cubic, leaveTangent=0.3205, tangentMode=RCTM_User, time=1, value=1}
        ]
       }
      }}))
   &UpdateMeshOrientation(Rotation Rate=&RandomRangeFloat(Minimum=0.05, Maximum=0.2)) {enabled=false}
   &SolveForcesAndVelocity(Clamp Velocity=false, Limit Acceleration=false)
   &DynamicMaterialParameters(
    Index 0 Param 1=&RandomRangeFloat(Minimum=0, Maximum=1)
    Index 0 Param 2=&RandomRangeFloat(Minimum=1.5, Maximum=5)
    Index 0 Param 4=&FloatFromCurve(
     FloatCurve=DataInterface() {propertyValues {Curve {defaultValue=1, keys [{interpMode=RCIM_Cubic, tangentMode=RCTM_User, time=0.25, value=0}, <at 1 1 RCIM_Cubic>]}}}
     Scale Curve=6
    )
    Module.Index0Float0Write=true
    Module.Index0Float1Write=true
    Module.Index0Float2Write=true
    Module.Index0Float3Write=true
   )
   &ScaleColor(
    Scale Alpha=&FloatFromCurve(FloatCurve=DataInterface() {propertyValues {Curve {
        defaultValue=1
        keys [
         {arriveTangent=6.4544, interpMode=RCIM_Cubic, time=0, value=0}
         {
          arriveTangent=0.4432
          arriveTangentWeight=0.1333
          interpMode=RCIM_Cubic
          leaveTangent=0.4432
          leaveTangentWeight=0.0667
          tangentMode=RCTM_User
          time=0.2
          value=0.9
         }
         {arriveTangentWeight=0.1296, interpMode=RCIM_Cubic, leaveTangentWeight=0.2592, tangentMode=RCTM_User, time=0.6, value=1}
         {tangentMode=RCTM_User, time=1, value=0}
        ]
       }}})
    ScaleA=true
    ScaleRGB=false
   )
  ]
  rendererClasses [&NiagaraMeshRendererProperties]
  renderers [{rendererClass=&NiagaraMeshRendererProperties, rendererIndex=0}]
 }
 ZP_Tendril_Ribbons {
  bEnabled=false
  emitterUpdateScript [
   &EmitterState(
    Inactive Response=::NewEnumerator1
    Life Cycle Mode=::NewEnumerator1
    Loop Behavior=::NewEnumerator1
    Loop Duration Mode=::NewEnumerator1
    UseLoopDelay=false
   )
   &SpawnBurst_Instantaneous(Spawn Count=$System.Spline1Count, Spawn Time=0.05, Use Spawn Probability=false, UseLoopCountLimit=false)
   &SpawnBurst_Instantaneous(Spawn Count=$System.Spline2Count, Spawn Group=1, Spawn Time=0.1, Use Spawn Probability=false, UseLoopCountLimit=false) {module=SpawnBurst_Instantaneous001}
   &SpawnBurst_Instantaneous(Spawn Count=$System.Spline3Count, Spawn Group=2, Spawn Time=0.15, Use Spawn Probability=false, UseLoopCountLimit=false) {module=SpawnBurst_Instantaneous002}
   &SpawnBurst_Instantaneous(Spawn Count=$System.Spline1Count, Spawn Group=3, Spawn Time=0.2, Use Spawn Probability=false, UseLoopCountLimit=false) {module=SpawnBurst_Instantaneous003}
   &SpawnBurst_Instantaneous(Spawn Count=$System.Spline2Count, Spawn Group=4, Spawn Time=0.25, Use Spawn Probability=false, UseLoopCountLimit=false) {module=SpawnBurst_Instantaneous004}
   &SpawnBurst_Instantaneous(Spawn Count=$System.Spline3Count, Spawn Group=5, Spawn Time=0.3, Use Spawn Probability=false, UseLoopCountLimit=false) {module=SpawnBurst_Instantaneous005}
   &SpawnBurst_Instantaneous(Spawn Count=$System.Spline1Count, Spawn Group=6, Spawn Time=0.35, Use Spawn Probability=false, UseLoopCountLimit=false) {module=SpawnBurst_Instantaneous006}
   &SpawnBurst_Instantaneous(Spawn Count=$System.Spline2Count, Spawn Group=7, Spawn Time=0.4, Use Spawn Probability=false, UseLoopCountLimit=false) {module=SpawnBurst_Instantaneous007}
   &SpawnBurst_Instantaneous(Spawn Count=$System.Spline3Count, Spawn Group=8, Spawn Time=0.45, Use Spawn Probability=false, UseLoopCountLimit=false) {module=SpawnBurst_Instantaneous008}
  ]
  particleSpawnScript [
   /SetVariables(Particles.InitialExecIndex:NiagaraFloat=&ReturnNormalizedExecIndex()) {isSetParameters=true, module=SetVariables_#2}
   /SetVariables(Particles.RibbonID=/Niagara/DynamicInputs/Meshes/MakeRibbonIDFromSpawnGroup()) {isSetParameters=true, module=SetVariables_#3}
   /SetVariables(Particles.SplineIndex:NiagaraInt32=/Niagara/DynamicInputs/Divide/Divide_Int(A:NiagaraInt32=&SpawnGroup(), B:NiagaraInt32=3)) {isSetParameters=true}
   /SetVariables(Particles.ParticleOffset:NiagaraFloat=&Add_Float(
     A=&Multiply_Float(
      A=&ReturnNormalizedExecIndex(Normalized Index Scale=&SelectFloatFromArray(Direct Array Index=$Particles.SplineIndex, Float Selection Array=DataInterface() {propertyValues {FloatData []}}))
      B=4
     )
     B=&Multiply_FloatByInt(Float=0.125, Integer=&SpawnGroup())
    )) {isSetParameters=true, module=SetVariables_#4}
   /SetVariables(Particles.Variation:NiagaraFloat=&Waveform(
     Global Phase=$Particles.ParticleOffset
     "Waveform [1]"=::NewEnumerator0
     "[1] Frequency"=&Multiply_Float(
      A=&SelectFloatFromArray(
       Array Sampling Mode=::NewEnumerator0
       Direct Array Index=?
       Float Selection Array=DataInterface() {propertyValues {FloatData []}}
       Random Seed=0
      )
      B=3
     )
     "[1] Amplitude Min/Max"=<0.25 1>
    )) {isSetParameters=true, module=SetVariables_#5}
   /SetVariables(Particles.EndFade:NiagaraFloat=&FloatFromCurve(
     FloatCurve=DataInterface() {propertyValues {Curve {keys [
         {arriveTangent=15.64, interpMode=RCIM_Cubic, leaveTangent=15.64, tangentMode=RCTM_User, time=0, value=0}
         {arriveTangentWeight=0.0555, interpMode=RCIM_Cubic, leaveTangentWeight=0.2989, tangentMode=RCTM_User, time=0.15, value=1}
         {arriveTangentWeight=0.2661, interpMode=RCIM_Cubic, leaveTangentWeight=0.1352, tangentMode=RCTM_User, time=0.7, value=1}
         <at 1 0 RCIM_Cubic>
        ]}}}
     CurveIndex=&ReturnNormalizedExecIndex()
    )) {isSetParameters=true, module=SetVariables_#6}
   &InitializeParticle(
    Color=&SelectLinearColorFromArray(
     Array Sampling Mode=::NewEnumerator1
     Direct Array Index=&SpawnGroup()
     Direct Array Mode=::NewEnumerator1
     Color Selection Array=DataInterface() {propertyValues {ColorData [<rgba 2 1.0247 1.276 1>, <rgba 0.7408 0.539 1 0.7>, <rgba 1.4483 1.5727 2 0.7>]}}
     Random Seed=?
    )
    Ribbon Width=&Lerp_Float(A=80, B=500, Alpha=$Particles.EndFade)
    Ribbon Width Mode=::NewEnumerator1
    Sprite Size Mode=::NewEnumerator3
    Uniform Sprite Size=&Multiply_Float(A=800, B=$Particles.Variation)
    UsePositionOffset=false
    Write Material Random=true
   )
   &self:SetPositionOnSpline(
    Spline1:NiagaraDataInterfaceSpline=$System.Spline1
    Spline2:NiagaraDataInterfaceSpline=$System.Spline2
    Spline3:NiagaraDataInterfaceSpline=$System.Spline3
    SplineIndex:NiagaraInt32=$Particles.SplineIndex
    float:NiagaraFloat=&FloatFromCurve(
     FloatCurve=DataInterface() {propertyValues {Curve {
        defaultValue=1
        keys [
         {arriveTangent=1.2091, interpMode=RCIM_Cubic, leaveTangent=1.2091, tangentMode=RCTM_User, time=0, value=0}
         {arriveTangent=0.1083, interpMode=RCIM_Cubic, leaveTangent=0.1083, tangentMode=RCTM_User, time=1, value=1}
        ]
       }}}
     CurveIndex=&ReturnNormalizedExecIndex()
    )
   ) {module=SetPositionOnSpline}
   /SetVariables(Particles.Direction:Vector3f=$Output.SetPositionOnSpline.SampledDirection) {isSetParameters=true, module=SetVariables_#7}
  ]
  particleUpdateScript [
   &ParticleState(Kill Particles When Lifetime Has Elapsed=false)
   &DynamicMaterialParameters(
    Index 0 Param 1=&Add_Float(
     A=&Multiply_FloatByInt(Float=20.33, Integer=&SpawnGroup())
     B=&Multiply_Float(
      A=&SelectFloatFromArray(
       Direct Array Index=&SpawnGroup()
       Direct Array Mode=::NewEnumerator1
       Float Selection Array=DataInterface() {propertyValues {FloatData [-0.25,-0.1,0]}}
      )
      B=$Particles.Age
     )
    )
    Index 0 Param 3=1.1
    Module.Index0Float0Write=true
    Module.Index0Float1Write=true
    Module.Index0Float2Write=true
    Module.Index0Float3Write=false
   )
   /SetVariables(Particles.AnimatedVariation:NiagaraFloat=&Waveform(
     Global Phase=&Add_Float(A=&Multiply_Float(A=$Particles.Age, B=0.02), B=$Particles.Variation)
     "Waveform [1]"=::NewEnumerator0
     "[1] Frequency"=&Multiply_Float(
      A=&SelectFloatFromArray(
       Array Sampling Mode=::NewEnumerator0
       Direct Array Index=?
       Float Selection Array=DataInterface() {propertyValues {FloatData []}}
       Random Seed=0
      )
      B=10
     )
     "[1] Amplitude Min/Max"=<0 1>
    )) {isSetParameters=true, module=SetVariables_#8}
   /SetVariables(Particles.Position=&AddVectorToPosition(
     Position=$Particles.Initial.Position
     Vector=&RotateVectorByQuaternion(
      Vector To Rotate=&MakeVector(
       X=0
       Y=0
       Z=&Multiply_Float(A=&Lerp_Float(A=150, B=600, Alpha=$Particles.EndFade), B=$Particles.Variation)
      )
      Quaternion=&MakeQuaternion(
       Quaternion From=0
       Axis=$Particles.Direction
       Angle=&Add_Float(
        A=&Multiply_Float(A=$Particles.ParticleOffset, B=2)
        B=&Multiply_Float(A=$Particles.Age, B=-0.05)
       )
       From=?
       To=?
       Angle Type=::NewEnumerator1
       Normalize Vectors=false
      )
     )
    )) {isSetParameters=true, module=SetVariables_#9}
   &ScaleRibbonWidth(Ribbon Width Scale=&Lerp_Float(A=0.2, B=1, Alpha=$Particles.AnimatedVariation))
   &ScaleColor(
    Scale Alpha=&Multiply_Float(
     A=&Lerp_Float(A=0, B=0.4, Alpha=$Particles.AnimatedVariation)
     B=&FloatFromCurve(
      FloatCurve=DataInterface() {propertyValues {Curve {
         defaultValue=1
         keys [
          {arriveTangent=6.3543, interpMode=RCIM_Cubic, time=0, value=0}
          {arriveTangent=-2.573, interpMode=RCIM_Cubic, leaveTangent=-2.573, tangentMode=RCTM_User, time=0.5, value=2}
          {tangentMode=RCTM_User, time=1, value=1}
         ]
        }}}
      CurveIndex=&Subtract_Float(A=$Particles.Age, B=&Multiply_Float(A=$Particles.InitialExecIndex, B=1))
     )
    )
    ScaleA=true
    ScaleRGB=false
   )
  ]
  rendererClasses [&NiagaraSpriteRendererProperties, &NiagaraRibbonRendererProperties]
  renderers [
   {rendererClass=&NiagaraSpriteRendererProperties}
   {rendererClass=&NiagaraRibbonRendererProperties, rendererIndex=1}
  ]
 }
 ZP_Tendril_Ribbons001 {
  emitterUpdateScript [
   &EmitterState(
    Inactive Response=::NewEnumerator1
    Life Cycle Mode=::NewEnumerator1
    Loop Behavior=::NewEnumerator1
    Loop Duration Mode=::NewEnumerator1
    UseLoopDelay=false
   )
   &SpawnBurst_Instantaneous(Spawn Count=16, Spawn Time=0.05, Use Spawn Probability=false, UseLoopCountLimit=false)
   &SpawnBurst_Instantaneous(Spawn Count=16, Spawn Group=1, Spawn Time=0.1, Use Spawn Probability=false, UseLoopCountLimit=false) {module=SpawnBurst_Instantaneous003}
   &SpawnBurst_Instantaneous(Spawn Count=22, Spawn Group=2, Spawn Time=0.15, Use Spawn Probability=false, UseLoopCountLimit=false) {module=SpawnBurst_Instantaneous006}
  ]
  particleSpawnScript [
   /SetVariables(Particles.InitialExecIndex:NiagaraFloat=&ReturnNormalizedExecIndex()) {isSetParameters=true, module=SetVariables_#2}
   /SetVariables(Particles.RibbonID=&MakeRibbonIDFromSpawnGroup()) {isSetParameters=true, module=SetVariables_#3}
   /SetVariables(Particles.ParticleOffset:NiagaraFloat=&Add_Float(
     A=&Multiply_Float(
      A=&ReturnNormalizedExecIndex(Normalized Index Scale=&SelectFloatFromArray(
        Direct Array Index=&SpawnGroup()
        Float Selection Array=DataInterface() {propertyValues {FloatData [1,0.8,0.9]}}
       ))
      B=4
     )
     B=&Multiply_FloatByInt(Float=0.125, Integer=&SpawnGroup())
    )) {isSetParameters=true, module=SetVariables_#4}
   /SetVariables(Particles.Variation:NiagaraFloat=&Waveform(
     Global Phase=$Particles.ParticleOffset
     "Waveform [1]"=::NewEnumerator0
     "[1] Frequency"=&Multiply_Float(
      A=&SelectFloatFromArray(
       Direct Array Index=&SpawnGroup()
       Float Selection Array=DataInterface() {propertyValues {FloatData [0.8,0.9,1]}}
      )
      B=15
     )
     "[1] Amplitude Min/Max"=<0.25 1>
    )) {isSetParameters=true, module=SetVariables_#5}
   /SetVariables(Particles.EndFade:NiagaraFloat=&FloatFromCurve(
     FloatCurve=DataInterface() {propertyValues {Curve {keys [
         {arriveTangent=15.64, interpMode=RCIM_Cubic, leaveTangent=15.64, tangentMode=RCTM_User, time=0, value=0}
         {arriveTangentWeight=0.0555, interpMode=RCIM_Cubic, leaveTangentWeight=0.2989, tangentMode=RCTM_User, time=0.15, value=1}
         {arriveTangentWeight=0.2661, interpMode=RCIM_Cubic, leaveTangentWeight=0.1352, tangentMode=RCTM_User, time=0.7, value=1}
         {arriveTangent=-4.1842, interpMode=RCIM_Cubic, leaveTangent=-4.1842, tangentMode=RCTM_User, time=1, value=0}
        ]}}}
     CurveIndex=&ReturnNormalizedExecIndex()
    )) {isSetParameters=true, module=SetVariables_#6}
   &InitializeParticle(
    Ribbon Width=&Lerp_Float(A=100, B=400, Alpha=$Particles.EndFade)
    Ribbon Width Mode=::NewEnumerator1
    Sprite Size Mode=::NewEnumerator3
    Uniform Sprite Size=&Multiply_Float(A=800, B=$Particles.Variation)
    UsePositionOffset=false
    Write Material Random=true
   )
   &self:SetPositionOnSpline(
    Spline1:NiagaraDataInterfaceSpline=$System.Spline1
    Spline2:NiagaraDataInterfaceSpline=$System.Spline2
    Spline3:NiagaraDataInterfaceSpline=$System.Spline3
    SplineIndex:NiagaraInt32=&SpawnGroup()
    float:NiagaraFloat=&FloatFromCurve(
     FloatCurve=DataInterface() {propertyValues {Curve {
        defaultValue=1
        keys [
         {arriveTangent=0.3204, interpMode=RCIM_Cubic, leaveTangent=0.3204, tangentMode=RCTM_User, time=0, value=0}
         {arriveTangent=1, interpMode=RCIM_Cubic, leaveTangent=1, time=0.1372, value=0.0988}
         {arriveTangent=0.1083, interpMode=RCIM_Cubic, leaveTangent=0.1083, tangentMode=RCTM_User, time=1, value=1}
        ]
       }}}
     CurveIndex=&ReturnNormalizedExecIndex()
    )
   ) {module=SetPositionOnSpline}
   /SetVariables(Particles.Direction:Vector3f=$Output.SetPositionOnSpline.SampledDirection) {isSetParameters=true, module=SetVariables_#7}
  ]
  particleUpdateScript [
   &ParticleState(Kill Particles When Lifetime Has Elapsed=false)
   &DynamicMaterialParameters(
    Index 0 Param 1=&Multiply_Float(A=-0.25, B=$Particles.Age)
    Index 0 Param 3=1.1
    Index 1 Param 1=&Add_Float(A=3.33, B=&Multiply_Float(A=0, B=$Particles.Age))
    Index 1 Param 3=1.1
    Index 2 Param 1=&Add_Float(A=8.4, B=&Multiply_Float(A=-0.1, B=$Particles.Age))
    Index 2 Param 3=1.1
    Module.Index0Float0Write=true
    Module.Index0Float1Write=false
    Module.Index0Float2Write=true
    Module.Index0Float3Write=false
    Module.Index1Float0Write=true
    Module.Index1Float1Write=false
    Module.Index1Float2Write=true
    Module.Index1Float3Write=false
    Module.Index2Float0Write=true
    Module.Index2Float1Write=false
    Module.Index2Float2Write=true
    Module.Index2Float3Write=false
    Write Parameter Index 1=::NewEnumerator1
    Write Parameter Index 2=::NewEnumerator1
   )
   /SetVariables(Particles.AnimatedVariation:NiagaraFloat=&Waveform(
     Global Phase=&Add_Float(
      A=&Multiply_Float(A=$Particles.Age, B=0.02)
      B=&ReturnNormalizedExecIndex(Normalized Index Scale=-4)
     )
     "Waveform [1]"=::NewEnumerator0
     "[1] Frequency"=4
     "[1] Amplitude Min/Max"=<0 1>
    )) {isSetParameters=true, module=SetVariables_#8}
   /SetVariables(Particles.Position=&AddVectorToPosition(
     Position=$Particles.Initial.Position
     Vector=&RotateVectorByQuaternion(
      Vector To Rotate=&MakeVector(
       X=0
       Y=0
       Z=&Multiply_Float(A=&Lerp_Float(A=150, B=500, Alpha=$Particles.EndFade), B=$Particles.Variation)
      )
      Quaternion=&MakeQuaternion(
       Quaternion From=0
       Axis=$Particles.Direction
       Angle=&Add_Float(
        A=&Multiply_Float(A=$Particles.ParticleOffset, B=2)
        B=&Multiply_Float(A=$Particles.Age, B=-0.07)
       )
       From=?
       To=?
       Angle Type=::NewEnumerator1
       Normalize Vectors=false
      )
     )
    )) {isSetParameters=true, module=SetVariables_#9}
   /SetVariables(Particles.Position2:NiagaraPosition=&AddVectorToPosition(
     Position=$Particles.Initial.Position
     Vector=&RotateVectorByQuaternion(
      Vector To Rotate=&MakeVector(
       X=0
       Y=0
       Z=&Multiply_Float(
        A=&Lerp_Float(A=-150, B=-550, Alpha=$Particles.EndFade)
        B=&Multiply_Float(A=$Particles.Variation, B=0.8)
       )
      )
      Quaternion=&MakeQuaternion(
       Quaternion From=0
       Axis=$Particles.Direction
       Angle=&Add_Float(
        A=&Multiply_Float(A=&Add_Float(A=$Particles.ParticleOffset, B=0.33), B=2)
        B=&Multiply_Float(A=$Particles.Age, B=-0.06)
       )
       From=?
       To=?
       Angle Type=::NewEnumerator1
       Normalize Vectors=false
      )
     )
    )) {isSetParameters=true}
   /SetVariables(Particles.Position3:NiagaraPosition=&AddVectorToPosition(
     Position=$Particles.Initial.Position
     Vector=&RotateVectorByQuaternion(
      Vector To Rotate=&MakeVector(
       X=0
       Y=0
       Z=&Multiply_Float(A=&Lerp_Float(A=100, B=300, Alpha=$Particles.EndFade), B=$Particles.Variation)
      )
      Quaternion=&MakeQuaternion(
       Quaternion From=0
       Axis=$Particles.Direction
       Angle=&Add_Float(
        A=&Multiply_Float(A=&Add_Float(A=$Particles.ParticleOffset, B=0.625), B=2)
        B=&Multiply_Float(A=$Particles.Age, B=-0.05)
       )
       From=?
       To=?
       Angle Type=::NewEnumerator1
       Normalize Vectors=false
      )
     )
    )) {isSetParameters=true}
   &ScaleRibbonWidth(Ribbon Width Scale=&Lerp_Float(A=0.2, B=1, Alpha=$Particles.AnimatedVariation))
   /SetVariables(Particles.Alpha:NiagaraFloat=&Multiply_Float(
     A=&Lerp_Float(A=0, B=0.4, Alpha=$Particles.AnimatedVariation)
     B=&FloatFromCurve(
      FloatCurve=DataInterface() {propertyValues {Curve {
         defaultValue=1
         keys [
          {arriveTangent=6.3543, interpMode=RCIM_Cubic, time=0, value=0}
          {arriveTangent=-2.573, interpMode=RCIM_Cubic, leaveTangent=-2.573, tangentMode=RCTM_User, time=0.5, value=2}
          {tangentMode=RCTM_User, time=1, value=1}
         ]
        }}}
      CurveIndex=&Subtract_Float(A=$Particles.Age, B=&Multiply_Float(A=$Particles.InitialExecIndex, B=1))
     )
    )) {isSetParameters=true}
   /SetVariables(
    Particles.Color=&MultiplyLinearColor(
     Linear Color=<rgba 2 1.0247 1.276 1>
     Scale Factor=&MakeVector4FromVectorAndFloat("Vector (RGB)"=<1 1 1>, "Float (Alpha)"=$Particles.Alpha)
    )
    Particles.Color2=&MultiplyLinearColor(
     Linear Color=<rgba 0.7408 0.539 1 0.7>
     Scale Factor=&MakeVector4FromVectorAndFloat("Vector (RGB)"=<1 1 1>, "Float (Alpha)"=$Particles.Alpha)
    )
    Particles.Color3:LinearColor=&MultiplyLinearColor(
     Linear Color=<rgba 1.2775 1.4987 2 0.7>
     Scale Factor=&MakeVector4FromVectorAndFloat("Vector (RGB)"=<1 1 1>, "Float (Alpha)"=$Particles.Alpha)
    )
   ) {isSetParameters=true}
  ]
  rendererClasses [&NiagaraSpriteRendererProperties, &NiagaraRibbonRendererProperties]
  renderers [
   {rendererClass=&NiagaraSpriteRendererProperties}
   {rendererClass=&NiagaraRibbonRendererProperties, rendererIndex=1}
   {rendererClass=&NiagaraRibbonRendererProperties, rendererIndex=2}
   {rendererClass=&NiagaraRibbonRendererProperties, rendererIndex=3}
  ]
 }
}
summary {
 systemName=NS_DragonCart_Shard_Heart
 emitters [
  {
   bEnabled=false
   emitterName=ZP_Tendril_Ribbons
   rendererClasses [&NiagaraSpriteRendererProperties, &NiagaraRibbonRendererProperties]
  }
  {emitterName=Splines, rendererClasses [&NiagaraMeshRendererProperties]}
  {
   emitterName=ZP_Tendril_Ribbons001
   rendererClasses [&NiagaraSpriteRendererProperties, &NiagaraRibbonRendererProperties]
  }
  {bEnabled=false, emitterName=Sparkles, simTarget=GPUComputeSim, rendererClasses [&NiagaraSpriteRendererProperties]}
 ]
 userVariables [
  User.Spline1 {
   type=Object
   description=""
   defaultValue {struct=NiagaraExt_VariableValue_Object, value {object=null, objectClass=&Object}}
  }
  User.Spline2 {
   type=Object
   description=""
   defaultValue {struct=NiagaraExt_VariableValue_Object, value {object=null, objectClass=&Object}}
  }
  User.Spline3 {
   type=Object
   description=""
   defaultValue {struct=NiagaraExt_VariableValue_Object, value {object=null, objectClass=&Object}}
  }
  User.ZPPosition {type=NiagaraPosition, description="", defaultValue {}}
 ]
}
