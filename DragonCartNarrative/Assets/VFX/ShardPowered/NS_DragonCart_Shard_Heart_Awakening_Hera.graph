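# Niagara system for the Hera DragonCart shard heart awakening: a purple glowing core amid floating shard meshes.
# Flare_Complex layers sprites and a light, while SolarFlares (toggled by User.SpawnSolarFlares) and Splines emit expanding, fading, distance-culled meshes.
# System.Intensity drives their size, brightness, spawn rates and motion.
# ShardMeshes partitions a burst into outer and inner shards that bob and rotate, with jitter from the NPC_DragonCart_Orb_Rift parameter collection.
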NiagaraSystem /DragonCartNarrative/Assets/VFX/ShardPowered/NS_DragonCart_Shard_Heart_Awakening_Hera
dependencies {
 usedDataInterfaces [
  &NiagaraDataInterfaceCurve
  &NiagaraDataInterfaceVectorCurve
  &NiagaraDataInterfaceMeshRendererInfo
 ]
 usedDynamicInputs [
  &ModuloFloat
  &FNTimeOfDayColor
  &FNTimeOfDayScalar
  &MakeLinearColor
 ]
 usedModules [
  &self:SystemSpawnScript.SystemScriptSource.SystemScriptGraph.NiagaraNodeAssignment_2.SetVariables_#1
  &SystemState
  &self:SystemSpawnScript.SystemScriptSource.SystemScriptGraph.NiagaraNodeAssignment_0.SetVariables_#2
  &self:SystemSpawnScript.SystemScriptSource.SystemScriptGraph.NiagaraNodeAssignment_1.SetVariables_#3
  &self:SystemSpawnScript.SystemScriptSource.SystemScriptGraph.NiagaraNodeAssignment_4.SetVariables_#4
  &self:SystemSpawnScript.SystemScriptSource.SystemScriptGraph.NiagaraNodeAssignment_3.SetVariables_#5
 ]
}
emitters {
 Flare_Complex {
  emitterUpdateScript [
   &EmitterState(
    Inactive Response=::NewEnumerator1
    Life Cycle Mode=::NewEnumerator1
    Loop Behavior=::NewEnumerator1
    Loop Duration Mode=::NewEnumerator1
    UseLoopDelay=false
   )
   &SpawnBurst_Instantaneous(Use Spawn Probability=false, UseLoopCountLimit=false)
  ]
  particleSpawnScript [
   &InitializeParticle(
    Color=<rgba 10 2.9993 8.3174 1>
    Lifetime=0.5
    Sprite Size Mode=::NewEnumerator3
    Uniform Sprite Size=2000
    UsePositionOffset=false
    Write Material Random=true
   )
   &CameraOffset(Camera Offset Amount=350)
   /SetVariables(Particles.WideFlareCamOffset=500) {isSetParameters=true}
  ]
  particleUpdateScript [
   &ParticleState(Kill Particles When Lifetime Has Elapsed=false)
   &ScaleSpriteSize(Scale Sprite Size Mode=::NewEnumerator0, Uniform Scale Factor=&Lerp_Float(A=0.75, B=1, Alpha=$System.Intensity))
   &DynamicMaterialParameters(
    Index 0 Param 1=0.7
    Index 0 Param 2=&Lerp_Float(A=0.6, B=0.8, Alpha=$System.Intensity)
    Index 3 Param 1=2
    Index 3 Param 2=5
    Module.Index0Param0Write=true
    Module.Index0Param1Write=true
    Module.Index0Param2Write=true
    Module.Index0Param3Write=false
    Module.Index3Param0Write=true
    Module.Index3Param1Write=true
    Module.Index3Param2Write=false
    Module.Index3Param3Write=false
    Write Parameter Index 3=true
   )
   &ScaleColor(
    Scale Alpha=&FloatFromCurve(FloatCurve=DataInterface() {propertyValues {Curve {keys [
         <at 0 0 RCIM_Cubic>
         {arriveTangentWeight=0.0943, interpMode=RCIM_Cubic, leaveTangentWeight=0.3771, tangentMode=RCTM_User, time=0.2, value=1}
         <at 1 1>
        ]}}})
    Scale RGB=&Multiply_VectorByFloat(
     Vector=&MakeVectorFromLinearColorRGB(Color=<rgba 0.5163 0.1091 1 1>)
     Float=&Lerp_Float(A=1, B=3, Alpha=$System.Intensity)
    )
    ScaleA=true
    ScaleRGB=true
   )
   /SetVariables(
    Particles.Darken Color:LinearColor=&MakeLinearColorFromVectorAndFloat("Vector (RGB)"=&MakeVectorFromLinearColorRGB(Color=<rgba 0.02 0.0024 0.0035 1>), "Float (Alpha)"=1.25)
    Particles.Darken Size:Vector2f=&Multiply_Vector2D_ByFloat(Float=1.25, Vector2D=$Particles.SpriteSize)
   ) {isSetParameters=true}
   /SetVariables(
    Particles.Glow Camera Offset=20
    Particles.Glow Color=&MultiplyLinearColor(
     Linear Color=<rgba 20 9.1032 1.8253 1>
     Scale Factor=&MakeFloat4FromVectorAndFloat(
      "Vector (RGB)"=&VectorFromFloat(Value=&Lerp_Float(A=1.2, B=1.5, Alpha=$System.Intensity))
      "Float (Alpha)"=1
     )
    )
    Particles.Glow Size=&Multiply_Vector2D_ByFloat(
     Float=&Waveform(Global Amplitude Scale=0.25, "Waveform [1]"=::NewEnumerator12, "[1] Frequency"=12, "[1] Amplitude Min/Max"=<0.84 1>)
     Vector2D=$Particles.SpriteSize
    )
   ) {isSetParameters=true}
   /SetVariables(
    Particles.Wide Flare Scale:Vector2f=&Multiply_Vector2D_ByFloat(Float=&Lerp_Float(A=1.25, B=2, Alpha=$System.Intensity), Vector2D=<2000 400>)
    Particles.WideFlareColor:LinearColor=&MultiplyLinearColor(
     Linear Color=<rgba 1.2242 0.4289 3 1>
     Scale Factor=&MakeFloat4FromVectorAndFloat(
      "Vector (RGB)"=&VectorFromFloat(Value=&Lerp_Float(A=1, B=2, Alpha=$System.Intensity))
      "Float (Alpha)"=5
     )
    )
   ) {isSetParameters=true}
   /SetVariables(
    Particles.LightColor=&MultiplyLinearColor(Linear Color=$Particles.Color, Scale Factor=<2 1 1 10>)
    Particles.LightRadius=&Lerp_Float(A=2800, B=4000, Alpha=$System.Intensity)
   ) {isSetParameters=true}
  ]
  rendererClasses [&NiagaraSpriteRendererProperties, &NiagaraLightRendererProperties]
  renderers [
   {rendererClass=&NiagaraSpriteRendererProperties}
   {rendererClass=&NiagaraSpriteRendererProperties, rendererIndex=1}
   {rendererClass=&NiagaraSpriteRendererProperties, rendererIndex=2}
   {rendererClass=&NiagaraSpriteRendererProperties, rendererIndex=3}
   {rendererClass=&NiagaraLightRendererProperties, rendererIndex=4}
  ]
 }
 Light_Fixture_Glows {rendererClasses [], renderers []}
 PlasmaWave {rendererClasses [], renderers []}
 ShardMeshes {
  emitterUpdateScript [
   &EmitterState(
    Enable Distance Culling=true
    Inactive Response=::NewEnumerator1
    Life Cycle Mode=::NewEnumerator1
    Loop Behavior=::NewEnumerator1
    Loop Duration Mode=::NewEnumerator1
    Max Distance Response=::SleepAndClearParticles
    MaxDistance=14000
    Module.bUseMaxDistance=true
    Module.bUseMinDistance=false
    Reset Age On Awaken=true
    Scalability Mode=::NewEnumerator1
    UseLoopDelay=false
   )
   &SpawnBurst_Instantaneous(Spawn Count=30, Use Spawn Probability=false, UseLoopCountLimit=false)
  ]
  particleSpawnScript [
   &PartitionParticles(Number Of Partitions=7, Partition Mode:ENiagara_PartitionMode_v2=::NewEnumerator5)
   /SetVariables(Particles.OuterShard:NiagaraBool=&SetBoolByIntComparison(A=$Particles.PartitionParticles.Partition, B=3, Comparison Type=::NewEnumerator0)) {isSetParameters=true}
   &InitializeParticle(
    Color=<rgba 0.7708 0.156 1 1>
    Lifetime=&RandomRangeFloat(Minimum=9, Maximum=12)
    Mesh Renderer Array Visibility Mode=::NewEnumerator1
    Mesh Renderer Info=DataInterface() {propertyValues {MeshRenderer=&self:ShardMeshes_0.NiagaraMeshRendererProperties_0}}
    Mesh Renderer Mesh Index:NiagaraInt32=$Particles.PartitionParticles.Partition
    Mesh Scale Mode=::NewEnumerator2
    UsePositionOffset=false
    Write Material Random=true
   )
   &ShapeLocation(
    MaskSpawnGroup=false
    Rotation Mode=::NewEnumerator2
    Sphere Distribution=::NewEnumerator1
    Sphere Radius=420
    Sphere Radius Position:NiagaraFloat=&MakeCustomFloatFromBool(Bool=$Particles.OuterShard, True Float=1, False Float=&RandomRangeFloat(Minimum=0.5, Maximum=0.7))
    U Position=&ReturnNormalizedExecIndex(Normalized Index Scale=&RandomRangeFloat(Minimum=1.9, Maximum=2))
    UseOwnerScale=true
    V Position=&ReturnNormalizedExecIndex(Normalized Index Scale=&RandomRangeFloat(Minimum=0.9, Maximum=1))
    Yaw / Pitch / Roll=<0 0.31 -0.17>
   )
   &InitialMeshOrientation(
    Mesh Orientation Mode=::NewEnumerator2
    Orientation Vector=&Multiply_VectorByFloat(Vector=Linked() {linkedVariable {name=Particles.Position, type=NiagaraPosition}}, Float=-1)
    Rotation=&RandomRangeVector(Minimum=<0 0 0>, Maximum=<1 0 0>)
    Use Rotation Vector=true
   )
  ]
  particleUpdateScript [
   &ParticleState(Kill Particles When Lifetime Has Elapsed=false)
   &UpdateMeshOrientation(
    Rotation Rate=&Add_Float(
     A=&MakeCustomFloatFromBool(
      Bool=$Particles.OuterShard
      True Float=&RandomRangeFloat(Minimum=-0.05, Maximum=0.05)
      False Float=&RandomRangeFloat(Minimum=0.05, Maximum=0.15)
     )
     B=&Multiply_Float(A=$NPC.NPC_DragonCart_Orb_Rift.ZPJitter, B=0.25)
    )
    Rotation Vector=&MakeCustomVectorFromBool(Bool=$Particles.OuterShard, True Vector=<1 0 0>, False Vector=&RandomVector(Vector Scale=1))
   )
   /SetVariables(Particles.Position=&AddVectorToPosition(
     Position=$Particles.Initial.Position
     Vector=&Add_Vector(
      A=&RotateVectorByQuaternion(
       Vector To Rotate=&RandomRangeVector(Minimum=<0 15 0>, Maximum=<0 0 15>)
       Quaternion=&MakeQuaternion(
        Quaternion From=0
        Axis=&NormalizeVector(
         Vector To Normalize=Linked() {linkedVariable {name=Particles.Initial.Position, type=NiagaraPosition}}
         Safe Normalize=true
         Safe Fallback Threshold=0
         Safe Fallback Vector=<0 0 0.001>
        )
        Angle=&Waveform(
         "Waveform [1]"=::NewEnumerator0
         "[1] Frequency"=&RandomRangeFloat(Minimum=0.08, Maximum=0.16)
         "[1] Amplitude Min/Max"=<0 1>
        )
        From=?
        To=?
        Angle Type=::NewEnumerator1
        Normalize Vectors=false
       )
      )
      B=&MakeCustomVectorFromBool(
       Bool=$Particles.OuterShard
       True Vector=<0 0 0>
       False Vector=&Multiply_VectorByFloat(
        Vector=&MakeVector(
         X=&Waveform(
          "Waveform [1]"=::NewEnumerator12
          "[1] Frequency"=&RandomRangeFloat(Minimum=7, Maximum=11)
          "[1] Amplitude Min/Max"=<-1 1>
         )
         Y=0
         Z=&Waveform(
          "Waveform [1]"=::NewEnumerator12
          "[1] Frequency"=&RandomRangeFloat(Minimum=7, Maximum=11)
          "[1] Amplitude Min/Max"=<-1 1>
         )
        )
        Float=&Multiply_Float(A=$NPC.NPC_DragonCart_Orb_Rift.ZPJitter, B=4)
       )
      )
     )
    )) {isSetParameters=true}
   &ScaleColor(Scale RGB=&VectorFromFloat(Value=1), ScaleA=false, ScaleRGB=true)
  ]
  rendererClasses [&NiagaraMeshRendererProperties]
  renderers [{rendererClass=&NiagaraMeshRendererProperties, rendererIndex=0}]
 }
 SolarFlares {
  emitterUpdateScript [
   &EmitterState(
    Enable Distance Culling=true
    MaxDistance=24000
    Module.bUseMaxDistance=true
    Module.bUseMinDistance=false
    Scalability Mode=::NewEnumerator1
   )
   &SpawnRate(
    SpawnRate=&MakeCustomFloatFromBool(Bool=$User.SpawnSolarFlares, True Float=&Lerp_Float(A=0.3, B=0.6, Alpha=$System.Intensity), False Float=0)
    Use Spawn Probability=false
   )
  ]
  particleSpawnScript [
   &InitializeParticle(
    Color Maximum=<rgba 2.0645 0.5149 2.5 1>
    Color Minimum=<rgba 0.6401 0.4516 1.5 1>
    Color Mode=::NewEnumerator2
    Lifetime=&Multiply_Float(
     A=&RandomRangeFloat(Minimum=10, Maximum=14)
     B=&Lerp_Float(A=1, B=0.5, Alpha=$System.Intensity)
    )
    Mesh Scale Max=<5 5 8>
    Mesh Scale Min=<5 1 5>
    Mesh Scale Mode=::NewEnumerator2
    UsePositionOffset=false
    Write Material Random=true
   )
   &ShapeLocation(MaskSpawnGroup=false, Override Seed=false, Sphere Radius=150, Sphere Surface Distribution=1, UseOwnerScale=true)
   &InitialMeshOrientation(
    Mesh Orientation Mode=::NewEnumerator2
    Rotation=&RandomRangeVector(Minimum=<0 0 0>, Maximum=<1 0 0>)
    Use Rotation Vector=true
   )
   &AddVelocity(
    Module.Use Falloff Distance=false
    Origin Offset=<0 0 -400>
    Origin Offset Coordinate Space=::World
    Override Seed=false
    Velocity Mode=::NewEnumerator1
    Velocity Speed=&RandomRangeFloat(Minimum=50, Maximum=100)
   )
  ]
  particleUpdateScript [
   &ParticleState()
   &ScaleMeshSize(Scale Factor=&VectorFromCurve(
     VectorCurve=DataInterface() {propertyValues {
       XCurve {
        defaultValue=1
        keys [{arriveTangent=0.9512, interpMode=RCIM_Cubic, time=0, value=0.1}, {interpMode=RCIM_Cubic, leaveTangent=0.4285, time=1, value=1}]
       }
       YCurve {
        defaultValue=1
        keys [
         {arriveTangent=-1.5843, interpMode=RCIM_Cubic, leaveTangent=-1.5843, tangentMode=RCTM_User, time=0, value=0.3}
         {arriveTangent=0.7, interpMode=RCIM_Cubic, leaveTangent=0.7, time=0.25, value=0.1}
         {arriveTangent=1.3734, interpMode=RCIM_Cubic, leaveTangent=1.3734, tangentMode=RCTM_User, time=1, value=1}
        ]
       }
       ZCurve {
        defaultValue=1
        keys [
         {arriveTangent=-1.5843, interpMode=RCIM_Cubic, leaveTangent=-1.5843, tangentMode=RCTM_User, time=0, value=0.3}
         {arriveTangent=0.7, interpMode=RCIM_Cubic, leaveTangent=0.7, time=0.25, value=0.1}
         {arriveTangent=1.3734, interpMode=RCIM_Cubic, leaveTangent=1.3734, tangentMode=RCTM_User, time=1, value=1}
        ]
       }
      }}
     Scale Curve=&Lerp_Vector(A=<0.7 0.9 0.9>, B=<1 1 1>, Alpha=$System.Intensity)
    ))
   &UpdateMeshOrientation(Rotation Rate=&FloatFromCurve(
     FloatCurve=DataInterface() {propertyValues {Curve {keys [{arriveTangent=-1.7586, interpMode=RCIM_Cubic, leaveTangent=-1.7586, tangentMode=RCTM_User, time=0, value=1}, <at 1 0.1>]}}}
     Scale Curve=&Multiply_Float(
      A=&RandomRangeFloat(Minimum=-0.15, Maximum=0.15)
      B=&Lerp_Float(A=1, B=2, Alpha=$System.Intensity)
     )
    ))
   &SolveForcesAndVelocity(Clamp Velocity=false, Limit Acceleration=false)
   &DynamicMaterialParameters(
    Index 0 Param 1=&RandomRangeFloat(Minimum=0, Maximum=1)
    Index 0 Param 2=&FloatFromCurve(
     FloatCurve=DataInterface() {propertyValues {Curve {keys [
         {arriveTangent=1.4423, interpMode=RCIM_Cubic, leaveTangent=1.4423, tangentMode=RCTM_User, time=-0.5066, value=0.5}
         {arriveTangent=-0.1659, interpMode=RCIM_Cubic, leaveTangent=-0.1659, time=0.6, value=1}
         <at 1 0.25 RCIM_Cubic>
        ]}}}
     Scale Curve=&RandomRangeFloat(Minimum=2, Maximum=3)
    )
    Index 0 Param 4=&FloatFromCurve(
     FloatCurve=DataInterface() {propertyValues {Curve {defaultValue=1, keys [{interpMode=RCIM_Cubic, tangentMode=RCTM_User, time=0.2, value=0}, <at 1 1 RCIM_Cubic>]}}}
     Scale Curve=2
    )
    Module.Index0Float0Write=true
    Module.Index0Float1Write=true
    Module.Index0Float2Write=true
    Module.Index0Float3Write=true
   )
   &ScaleColor(
    Scale Alpha=&FloatFromCurve(
     FloatCurve=DataInterface() {propertyValues {Curve {
        defaultValue=1
        keys [
         {arriveTangent=6.4544, interpMode=RCIM_Cubic, time=0, value=0}
         {
          arriveTangent=0.088
          arriveTangentWeight=0.1333
          interpMode=RCIM_Cubic
          leaveTangent=0.088
          leaveTangentWeight=0.0667
          tangentMode=RCTM_User
          time=0.1
          value=1
         }
         {arriveTangent=-1.0836, tangentMode=RCTM_User, time=1, value=0}
        ]
       }}}
     Scale Curve=&Lerp_Float(A=0.5, B=1, Alpha=$System.Intensity)
    )
    Scale RGB=&Lerp_Vector(
     A=<1 1 1>
     B=<5 12 1>
     Alpha=&FloatFromCurve(FloatCurve=DataInterface() {propertyValues {Curve {keys [<at 0 1 RCIM_Cubic>,<at 0.4 0>]}}})
    )
    ScaleA=true
    ScaleRGB=true
   )
  ]
  rendererClasses [&NiagaraMeshRendererProperties]
  renderers [{rendererClass=&NiagaraMeshRendererProperties, rendererIndex=0}]
 }
 Splines {
  emitterUpdateScript [
   &EmitterState(
    Enable Distance Culling=true
    MaxDistance=17000
    Module.bUseMaxDistance=true
    Module.bUseMinDistance=false
    Scalability Mode=::NewEnumerator1
   )
   &SpawnRate(SpawnRate=&Lerp_Float(A=4, B=10, Alpha=$System.Intensity), Use Spawn Probability=false)
  ]
  particleSpawnScript [
   &InitializeParticle(
    Color Maximum=<rgba 2.0645 0.5149 2.5 1>
    Color Minimum=<rgba 0.6401 0.4516 1.5 1>
    Color Mode=::NewEnumerator2
    Lifetime=&Multiply_Float(
     A=&RandomRangeFloat(Minimum=4, Maximum=6)
     B=&Lerp_Float(A=1, B=0.5, Alpha=$System.Intensity)
    )
    Mesh Scale Max=<40 40 50>
    Mesh Scale Min=<15 15 30>
    Mesh Scale Mode=::NewEnumerator2
    UsePositionOffset=false
    Write Material Random=true
   )
   &ShapeLocation(MaskSpawnGroup=false, Override Seed=false, Sphere Radius=50, Sphere Surface Distribution=1, UseOwnerScale=true)
   &InitialMeshOrientation(
    Mesh Orientation Mode=::NewEnumerator2
    Rotation=&RandomRangeVector(Minimum=<0 0 0>, Maximum=<1 0 0>)
    Use Rotation Vector=true
   )
   &AddVelocity(
    Module.Use Falloff Distance=false
    Override Seed=false
    Velocity Mode=::NewEnumerator1
    Velocity Speed=&Lerp_Float(A=150, B=400, Alpha=$System.Intensity)
   )
  ]
  particleUpdateScript [
   &ParticleState()
   &ScaleMeshSize(Scale Factor=&VectorFromCurve(
     VectorCurve=DataInterface() {propertyValues {
       XCurve {
        defaultValue=1
        keys [
         {arriveTangent=0.9512, interpMode=RCIM_Cubic, leaveTangent=0.9512, tangentMode=RCTM_User, time=0, value=0.2}
         {arriveTangent=0.4285, interpMode=RCIM_Cubic, leaveTangent=0.4285, tangentMode=RCTM_User, time=1, value=1}
        ]
       }
       YCurve {
        defaultValue=1
        keys [
         {arriveTangent=1.3531, interpMode=RCIM_Cubic, time=0, value=0.25}
         {arriveTangent=0.3205, interpMode=RCIM_Cubic, leaveTangent=0.3205, tangentMode=RCTM_User, time=1, value=1}
        ]
       }
       ZCurve {
        defaultValue=1
        keys [
         {arriveTangent=1.3531, interpMode=RCIM_Cubic, time=0, value=0.25}
         {arriveTangent=0.3205, interpMode=RCIM_Cubic, leaveTangent=0.3205, tangentMode=RCTM_User, time=1, value=1}
        ]
       }
      }}
     Scale Curve=&Lerp_Vector(A=<1 1 1>, B=<0.6 0.6 1.5>, Alpha=$System.Intensity)
    ))
   &UpdateMeshOrientation(Rotation Rate=&FloatFromCurve(
     FloatCurve=DataInterface() {propertyValues {Curve {keys [<at 0 1 RCIM_Cubic>,<at 1 0.1>]}}}
     Scale Curve=&Multiply_Float(
      A=&RandomRangeFloat(Minimum=-0.15, Maximum=0.15)
      B=&Lerp_Float(A=1, B=5, Alpha=$System.Intensity)
     )
    ))
   &SolveForcesAndVelocity(Clamp Velocity=false, Limit Acceleration=false)
   &DynamicMaterialParameters(
    Index 0 Param 1=&RandomRangeFloat(Minimum=0, Maximum=1)
    Index 0 Param 2=&FloatFromCurve(
     FloatCurve=DataInterface() {propertyValues {Curve {keys [<at 0 1 RCIM_Cubic>, {arriveTangent=-2.7759, tangentMode=RCTM_User, time=1, value=0.25}]}}}
     Scale Curve=&RandomRangeFloat(Minimum=2, Maximum=3)
    )
    Index 0 Param 4=&FloatFromCurve(
     FloatCurve=DataInterface() {propertyValues {Curve {defaultValue=1, keys [{interpMode=RCIM_Cubic, tangentMode=RCTM_User, time=0, value=0}, <at 1 1 RCIM_Cubic>]}}}
     Scale Curve=2
    )
    Module.Index0Float0Write=true
    Module.Index0Float1Write=true
    Module.Index0Float2Write=true
    Module.Index0Float3Write=true
   )
   &ScaleColor(
    Scale Alpha=&FloatFromCurve(
     FloatCurve=DataInterface() {propertyValues {Curve {
        defaultValue=1
        keys [
         {arriveTangent=6.4544, interpMode=RCIM_Cubic, time=0, value=0}
         {
          arriveTangent=0.088
          arriveTangentWeight=0.1333
          interpMode=RCIM_Cubic
          leaveTangent=0.088
          leaveTangentWeight=0.0667
          tangentMode=RCTM_User
          time=0.1
          value=1
         }
         {tangentMode=RCTM_User, time=1, value=0}
        ]
       }}}
     Scale Curve=&Lerp_Float(A=1, B=1.5, Alpha=$System.Intensity)
    )
    Scale RGB=&Lerp_Vector(
     A=<1 1 1>
     B=<5 10 1>
     Alpha=&FloatFromCurve(FloatCurve=DataInterface() {propertyValues {Curve {keys [<at 0 1 RCIM_Cubic>,<at 0.4 0>]}}})
    )
    ScaleA=true
    ScaleRGB=true
   )
  ]
  rendererClasses [&NiagaraMeshRendererProperties]
  renderers [{rendererClass=&NiagaraMeshRendererProperties, rendererIndex=0}]
 }
 Tendrils {rendererClasses [], renderers []}
}
summary {
 systemName=NS_DragonCart_Shard_Heart_Awakening_Hera
 emitters [
  {emitterName=Flare_Complex, rendererClasses [&NiagaraSpriteRendererProperties, &NiagaraLightRendererProperties]}
  {emitterName=Splines, rendererClasses [&NiagaraMeshRendererProperties]}
  {emitterName=PlasmaWave, rendererClasses []}
  {emitterName=ShardMeshes, rendererClasses [&NiagaraMeshRendererProperties]}
  {emitterName=SolarFlares, rendererClasses [&NiagaraMeshRendererProperties]}
  {emitterName=Light_Fixture_Glows, rendererClasses []}
  {emitterName=Tendrils, rendererClasses []}
 ]
 userVariables [
  User.Intensity {type=NiagaraFloat, defaultValue=0, description=""}
  User.SpawnSolarFlares {type=NiagaraBool, defaultValue=true}
 ]
}
