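# Niagara one-shot burst effect for the DragonCart shard heart, built as a large energy explosion.
# A glow sprite with a color-shifting light, an expanding mesh refraction sphere, outward-flying spline meshes and noisy, speed-stretched GPU sparkles form the burst.
# Curves drive scale, color and dynamic material parameters, while camera LOD distance adjusts sphere and sparkle parameters and splines are distance-culled.
# PlasmaWave has no renderers, and User.Intensity is exposed without visible use.
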NiagaraSystem /DragonCartNarrative/Assets/VFX/ShardPowered/NS_DragonCart_Shard_Heart_Burst
dependencies {
 usedDataInterfaces [&NiagaraDataInterfaceCurve, &NiagaraDataInterfaceColorCurve, &NiagaraDataInterfaceVectorCurve]
 usedDynamicInputs [&Add_Float, &ModuloFloat]
 usedModules [&SystemState]
}
emitters {
 Minimal {
  emitterUpdateScript [
   &EmitterState()
   &SpawnBurst_Instantaneous(Use Spawn Probability=false, UseLoopCountLimit=false)
  ]
  particleSpawnScript [&InitializeParticle(
    Color=<rgba 16 16 16 1>
    Lifetime=1.5
    Mass=700
    Mass Mode=::NewEnumerator1
    Sprite Size Mode=::NewEnumerator3
    Uniform Sprite Size=5000
    UsePositionOffset=false
    Write Material Random=true
   )]
  particleUpdateScript [
   &ParticleState()
   &ScaleSpriteSize(Uniform Curve Sprite Scale=DataInterface() {propertyValues {Curve {keys [
        {arriveTangent=1.7774, interpMode=RCIM_Cubic, leaveTangent=1.7774, tangentMode=RCTM_User, time=0, value=0.05}
        {arriveTangent=0.3956, tangentMode=RCTM_User, time=1, value=1}
       ]}}})
   &ScaleColor(
    Scale Alpha=&FloatFromCurve(FloatCurve=DataInterface() {propertyValues {Curve {
        defaultValue=1
        keys [
         {arriveTangent=18.8836, interpMode=RCIM_Cubic, leaveTangent=18.8836, tangentMode=RCTM_User, time=0, value=0}
         {arriveTangent=-2.1663, interpMode=RCIM_Cubic, leaveTangent=-2.1663, tangentMode=RCTM_User, time=0.1, value=1}
         {tangentMode=RCTM_User, time=0.8, value=0}
        ]
       }}})
    Scale RGB=&MakeVectorFromLinearColorRGB(Color=&ColorFromCurve(ColorCurve=DataInterface() {propertyValues {
        AlphaCurve {keys [<at 0 1 RCIM_Cubic>]}
        BlueCurve {keys [<at 0.1914 0.0157 RCIM_Cubic>,<at 0.4853 0.8754 RCIM_Cubic>]}
        GreenCurve {keys [<at 0.1914 0.7326 RCIM_Cubic>,<at 0.4853 0.3123 RCIM_Cubic>]}
        RedCurve {keys [<at 0.1914 2 RCIM_Cubic>,<at 0.4853 1 RCIM_Cubic>]}
       }}))
    ScaleA=true
    ScaleRGB=true
   )
   /SetVariables(Particles.LightColor=&ColorFromCurve(ColorCurve=DataInterface() {propertyValues {
       AlphaCurve {keys [<at 0 0 RCIM_Cubic>,<at 0.1 1 RCIM_Cubic>,<at 1 0 RCIM_Cubic>]}
       BlueCurve {keys [<at 0 0.7294 RCIM_Cubic>,<at 0.5904 13.6534 RCIM_Cubic>]}
       GreenCurve {keys [<at 0 8.262 RCIM_Cubic>,<at 0.5904 1.3281 RCIM_Cubic>]}
       RedCurve {keys [<at 0 25 RCIM_Cubic>,<at 0.5904 15 RCIM_Cubic>]}
      }})) {isSetParameters=true}
   &CameraOffset(Camera Offset Amount=/Niagara/DynamicInputs/Math/CalculateParticleRadius(
     Radius Calculation Type:ENiagaraCollisionRadiusOptions=::NewEnumerator0
     Particle Radius:NiagaraFloat=1
     Mesh Dimensions:Vector3f=<1 1 1>
     Method for Calculating Particle Radius:ENiagaraCalculateRadiusOptions=::NewEnumerator2
    ))
   &DynamicMaterialParameters(
    Index 0 Param 1=1
    Index 0 Param 2=&FloatFromCurve(FloatCurve=DataInterface() {propertyValues {Curve {keys [
         {arriveTangent=-2.2774, interpMode=RCIM_Cubic, leaveTangent=-2.2774, tangentMode=RCTM_User, time=0, value=1.5}
         {interpMode=RCIM_Cubic, leaveTangent=-0.6072, time=1, value=0.25}
        ]}}})
    Module.Index0Float0Write=true
    Module.Index0Float1Write=true
    Module.Index0Float2Write=false
    Module.Index0Float3Write=false
   )
  ]
  rendererClasses [&NiagaraSpriteRendererProperties, &NiagaraLightRendererProperties]
  renderers [
   {rendererClass=&NiagaraSpriteRendererProperties}
   {rendererClass=&NiagaraLightRendererProperties, rendererIndex=1}
  ]
 }
 PlasmaWave {rendererClasses [], renderers []}
 RefractionSphere {
  emitterUpdateScript [
   &EmitterState(Life Cycle Mode=::NewEnumerator1, Loop Behavior=::NewEnumerator1, Loop Duration=1, UseLoopDelay=false)
   &SpawnBurst_Instantaneous(Spawn Time=0.15, Use Spawn Probability=false, UseLoopCountLimit=false)
  ]
  particleSpawnScript [
   &InitializeParticle(
    Color=<rgba 1 1 1 0.75>
    Lifetime=1.5
    Mesh Scale Mode=::NewEnumerator3
    Mesh Uniform Scale=220
    UsePositionOffset=false
    Write Material Random=true
   )
   &InitialMeshOrientation(Mesh Axis To Orient=<0 0 1>, Mesh Orientation Mode=::NewEnumerator2, Use Rotation Vector=false)
  ]
  particleUpdateScript [
   &ParticleState()
   &ScaleMeshSize(Scale Factor=&VectorFromFloat(Value=&FloatFromCurve(
      FloatCurve=DataInterface() {propertyValues {Curve {defaultValue=0, keys [{tangentMode=RCTM_User, time=0, value=0}, {tangentMode=RCTM_User, time=1, value=1}]}}}
      Scale Curve=&FloatFromCurve(FloatCurve=DataInterface() {propertyValues {Curve {keys [<at 4000 0.3>,<at 50000 1>]}}}, CurveIndex=$Engine.Owner.LODDistance)
     )))
   &ScaleColor(
    Linear Color Curve=DataInterface() {propertyValues {
      AlphaCurve {keys [<at 0.4 1 RCIM_Cubic>, {arriveTangent=-2.2959, tangentMode=RCTM_User, time=1, value=0}]}
      BlueCurve {keys [<at 0.0195 0 RCIM_Cubic>,<at 0.562 1.1438>]}
      GreenCurve {keys [<at 0.0195 0.784 RCIM_Cubic>,<at 0.562 0.3255>]}
      RedCurve {keys [<at 0.0195 2 RCIM_Cubic>,<at 0.562 2>]}
     }}
    Scale Mode=::NewEnumerator2
   )
   &DynamicMaterialParameters(
    Index 0 Param 1=&Lerp_Float(
     A=1
     B=0.89
     Alpha=&FloatFromCurve(
      FloatCurve=DataInterface() {propertyValues {Curve {
         defaultValue=1
         keys [
          {interpMode=RCIM_Cubic, leaveTangentWeight=0.0333, tangentMode=RCTM_User, time=0, value=0}
          {
           arriveTangent=-1.6584
           arriveTangentWeight=0.0646
           interpMode=RCIM_Cubic
           leaveTangent=-1.6584
           leaveTangentWeight=0.581
           tangentMode=RCTM_User
           time=0.1
           value=1
          }
          {arriveTangentWeight=0.3, interpMode=RCIM_Cubic, tangentMode=RCTM_User, time=1, value=0}
         ]
        }}}
      Scale Curve=&FN_LOD_Distance_Remap(Start Distance=5000, End Distance=20000, Minimum Output=1, Maximum Output=0, Clamp Output=true)
     )
    )
    Index 0 Param 2=&FloatFromCurve(FloatCurve=DataInterface() {propertyValues {Curve {
        defaultValue=0
        keys [
         {arriveTangent=-2.0115, interpMode=RCIM_Cubic, leaveTangent=-2.0115, tangentMode=RCTM_User, time=0, value=1}
         {arriveTangent=-0.7166, interpMode=RCIM_Cubic, leaveTangent=-0.7166, tangentMode=RCTM_User, time=1, value=0}
        ]
       }}})
    Index 0 Param 3=&Lerp_Float(
     A=1
     B=0.5
     Alpha=&FloatFromCurve(FloatCurve=DataInterface() {propertyValues {Curve {defaultValue=1, keys [{arriveTangent=1.7704, interpMode=RCIM_Cubic, time=0, value=0}, {tangentMode=RCTM_User, time=1, value=1}]}}})
    )
    Index 0 Param 4=3000
    Module.Index0Float0Write=true
    Module.Index0Float1Write=true
    Module.Index0Float2Write=true
    Module.Index0Float3Write=true
   )
  ]
  rendererClasses [&NiagaraMeshRendererProperties]
  renderers [{rendererClass=&NiagaraMeshRendererProperties, rendererIndex=0}]
 }
 Sparkles {
  simTarget=GPUComputeSim
  emitterUpdateScript [
   &EmitterState(Life Cycle Mode=::NewEnumerator1, Loop Behavior=::NewEnumerator1, Loop Duration Mode=::NewEnumerator1, UseLoopDelay=false)
   &SpawnRate(
    SpawnRate=&FloatFromCurve(
     FloatCurve=DataInterface() {propertyValues {Curve {keys [{arriveTangent=-1.8943, interpMode=RCIM_Cubic, time=0, value=0}, <at 0.15 1 RCIM_Cubic>, <at 1 0>]}}}
     CurveIndex=$Emitter.Age
     Scale Curve=1500
    )
    Use Spawn Probability=false
   )
   &SpawnBurst_Instantaneous(Spawn Count=100, Spawn Time=0.15, Use Spawn Probability=false, UseLoopCountLimit=false)
  ]
  particleSpawnScript [
   &InitializeParticle(
    Color Maximum=<rgba 12 3.7157 0.46 1>
    Color Minimum=<rgba 5 0.1346 4.2232 1>
    Color Mode=::NewEnumerator2
    Lifetime Max=7
    Lifetime Min=4
    Lifetime Mode=::NewEnumerator1
    Sprite Size Mode=::NewEnumerator1
    Uniform Sprite Size Max=150
    Uniform Sprite Size Min=50
    UsePositionOffset=false
    Write Material Random=true
   )
   &ShapeLocation(Hemisphere Distribution=<1 0.7>, MaskSpawnGroup=false, Override Seed=false, Sphere Radius=300, UseOwnerScale=true)
   &AddVelocity(
    Module.Use Falloff Distance=false
    Override Seed=false
    Velocity Mode=::NewEnumerator1
    Velocity Speed=&Multiply_Float(
     A=&Lerp_Float(
      A=500
      B=8000
      Alpha=&FloatFromCurve(
       FloatCurve=DataInterface() {propertyValues {Curve {
          defaultValue=1
          keys [
           {arriveTangent=7.4531, interpMode=RCIM_Cubic, leaveTangent=7.4531, tangentMode=RCTM_User, time=0, value=0.4}
           {arriveTangent=-1.9963, interpMode=RCIM_Cubic, leaveTangent=-1.9963, tangentMode=RCTM_User, time=0.15, value=1}
           {tangentMode=RCTM_User, time=1, value=0}
          ]
         }}}
       CurveIndex=$Emitter.Age
       Scale Curve=&RandomRangeFloat(Minimum=0.7, Maximum=1)
      )
     )
     B=&FloatFromCurve(FloatCurve=DataInterface() {propertyValues {Curve {keys [<at 15000 1>,<at 30000 2>]}}}, CurveIndex=$Engine.Owner.LODDistance)
    )
   )
   /SetVariables(Particles.Lifetime=&Multiply_Float(A=&VectorLength(Vector=$Particles.Velocity), B=0.001)) {isSetParameters=true}
  ]
  particleUpdateScript [
   &ParticleState()
   &DynamicMaterialParameters(
    Index 0 Param 1=&FloatFromCurve(
     FloatCurve=DataInterface() {propertyValues {Curve {keys [{arriveTangent=-1.4828, interpMode=RCIM_Cubic, leaveTangent=-1.4828, tangentMode=RCTM_User, time=0, value=1}, <at 1 0>]}}}
     Scale Curve=&RandomRangeFloat(Minimum=0, Maximum=1)
    )
    Index 0 Param 2=&RandomRangeFloat(Minimum=0.5, Maximum=0.8)
    Module.Index0Param0Write=true
    Module.Index0Param1Write=true
    Module.Index0Param2Write=false
    Module.Index0Param3Write=false
   )
   &ScaleSpriteSize(
    Uniform Curve Scale=&FloatFromCurve(FloatCurve=DataInterface() {propertyValues {Curve {keys [<at 8000 0.25>,<at 30000 1>]}}}, CurveIndex=$Engine.Owner.LODDistance)
    Uniform Curve Sprite Scale=DataInterface() {propertyValues {Curve {keys [
        {arriveTangent=68.9421, interpMode=RCIM_Cubic, leaveTangent=68.9421, tangentMode=RCTM_User, time=0, value=0}
        {interpMode=RCIM_Cubic, tangentMode=RCTM_User, time=0.05, value=1.25}
        {arriveTangentWeight=0.0498, interpMode=RCIM_Cubic, leaveTangentWeight=0.3986, tangentMode=RCTM_User, time=0.15, value=1}
        {arriveTangent=-3.1615, tangentMode=RCTM_User, time=1, value=0}
       ]}}}
   )
   &Drag(
    Drag=&FloatFromCurve(
     FloatCurve=DataInterface() {propertyValues {Curve {keys [
         {arriveTangent=4.7895, interpMode=RCIM_Cubic, leaveTangent=4.7895, tangentMode=RCTM_User, time=0, value=0.1}
         {arriveTangent=0.3, interpMode=RCIM_Cubic, leaveTangent=0.3, time=0.25, value=1}
         {interpMode=RCIM_Cubic, leaveTangent=-4.8446, time=0.5, value=0.25}
        ]}}}
     Scale Curve=&RandomRangeFloat(Minimum=1, Maximum=2.5)
    )
    Use Linear Drag=true
    Use Rotational Drag=false
   )
   &VectorNoiseForce(Force Amount=4000, Module.Use Random Noise Time Interval=false)
   &SolveForcesAndVelocity(Clamp Velocity=false, Limit Acceleration=false)
   &ScaleSpriteSizeBySpeed(
    Max Scale Factor=<0.5 10>
    Sample Scale Factor By Curve=true
    Scale Factor Curve=DataInterface() {propertyValues {Curve {keys [<at 0.1 0 RCIM_Cubic>, {arriveTangent=1.1788, tangentMode=RCTM_User, time=1, value=1}]}}}
    Velocity Threshold:NiagaraFloat=&FloatFromCurve(
     FloatCurve=DataInterface() {propertyValues {Curve {keys [<at 15000 1>,<at 30000 2>]}}}
     CurveIndex=$Engine.Owner.LODDistance
     Scale Curve=8000
    )
   )
   &ScaleColor(
    Scale Alpha=&FloatFromCurve(FloatCurve=DataInterface() {propertyValues {Curve {keys [<at 0.5 1 RCIM_Cubic>, {arriveTangent=-4.8071, tangentMode=RCTM_User, time=1, value=0}]}}})
    Scale RGB=&VectorFromFloat(Value=&FloatFromCurve(FloatCurve=DataInterface() {propertyValues {Curve {keys [{arriveTangent=-13.9762, interpMode=RCIM_Cubic, leaveTangent=-13.9762, tangentMode=RCTM_User, time=0, value=5}, <at 0.5 1>]}}}))
    ScaleA=true
    ScaleRGB=true
   )
  ]
  rendererClasses [&NiagaraSpriteRendererProperties]
  renderers [{rendererClass=&NiagaraSpriteRendererProperties}]
 }
 Splines {
  emitterUpdateScript [
   &EmitterState(
    Enable Distance Culling=true
    MaxDistance=40000
    Module.bUseMaxDistance=true
    Module.bUseMinDistance=false
    Scalability Mode=::NewEnumerator1
   )
   &SpawnBurst_Instantaneous(Spawn Count=16, Use Spawn Probability=false, UseLoopCountLimit=false)
   &SpawnRate(
    SpawnRate=&FloatFromCurve(
     FloatCurve=DataInterface() {propertyValues {Curve {
        defaultValue=1
        keys [
         {arriveTangent=-0.5737, interpMode=RCIM_Cubic, leaveTangent=-0.5737, tangentMode=RCTM_User, time=0, value=1}
         {arriveTangent=-3.5855, tangentMode=RCTM_User, time=0.4, value=0}
        ]
       }}}
     CurveIndex=$Emitter.NormalizedLoopAge
     Scale Curve=30
    )
    Use Spawn Probability=false
   )
  ]
  particleSpawnScript [
   &InitializeParticle(
    Color Maximum=<rgba 5 2.2482 2.3437 1>
    Color Minimum=<rgba 2 2 2 1>
    Color Mode=::NewEnumerator2
    Lifetime Max=3
    Lifetime Min=2
    Lifetime Mode=::NewEnumerator1
    Mesh Scale Max=<90 90 140>
    Mesh Scale Min=<50 50 80>
    Mesh Scale Mode=::NewEnumerator2
    UsePositionOffset=false
    Write Material Random=true
   )
   &ShapeLocation(MaskSpawnGroup=false, Override Seed=false, Sphere Radius=50, Sphere Surface Distribution=1, UseOwnerScale=true)
   &InitialMeshOrientation(
    Mesh Orientation Mode=::NewEnumerator2
    Rotation=&RandomRangeVector(Minimum=<0 0 0>, Maximum=<1 0 0>)
    Use Rotation Vector=true
   )
   &AddVelocity(
    Module.Use Falloff Distance=false
    Override Seed=false
    Velocity Mode=::NewEnumerator1
    Velocity Speed=&RandomRangeFloat(Minimum=900, Maximum=1600)
   )
  ]
  particleUpdateScript [
   &ParticleState()
   &ScaleMeshSize(Scale Factor=&VectorFromCurve(VectorCurve=DataInterface() {propertyValues {
       XCurve {
        defaultValue=1
        keys [
         {arriveTangent=0.9512, interpMode=RCIM_Cubic, time=0, value=0}
         {arriveTangent=0.4285, interpMode=RCIM_Cubic, leaveTangent=0.4285, tangentMode=RCTM_User, time=1, value=1}
        ]
       }
       YCurve {
        defaultValue=1
        keys [
         {arriveTangent=1.3531, interpMode=RCIM_Cubic, time=0, value=0}
         {arriveTangent=0.3205, interpMode=RCIM_Cubic, leaveTangent=0.3205, tangentMode=RCTM_User, time=1, value=1}
        ]
       }
       ZCurve {
        defaultValue=1
        keys [
         {arriveTangent=2.2146, interpMode=RCIM_Cubic, leaveTangent=2.2146, tangentMode=RCTM_User, time=0, value=0}
         {arriveTangent=0.3205, interpMode=RCIM_Cubic, leaveTangent=0.3205, tangentMode=RCTM_User, time=1, value=1}
        ]
       }
      }}))
   &UpdateMeshOrientation(Rotation Rate=&FloatFromCurve(
     FloatCurve=DataInterface() {propertyValues {Curve {keys [<at 0 1 RCIM_Cubic>,<at 1 0.1>]}}}
     Scale Curve=&RandomRangeFloat(Minimum=-0.15, Maximum=0.15)
    ))
   &Drag(Drag=0.5, Use Linear Drag=true, Use Rotational Drag=false)
   &SolveForcesAndVelocity(Clamp Velocity=false, Limit Acceleration=false)
   &DynamicMaterialParameters(
    Index 0 Param 1=&RandomRangeFloat(Minimum=0, Maximum=1)
    Index 0 Param 2=&FloatFromCurve(
     FloatCurve=DataInterface() {propertyValues {Curve {keys [<at 0 1 RCIM_Cubic>, {arriveTangent=-2.7759, tangentMode=RCTM_User, time=1, value=0.25}]}}}
     Scale Curve=&RandomRangeFloat(Minimum=3, Maximum=5)
    )
    Index 0 Param 4=&FloatFromCurve(
     FloatCurve=DataInterface() {propertyValues {Curve {defaultValue=1, keys [{interpMode=RCIM_Cubic, tangentMode=RCTM_User, time=0.1, value=0}, <at 1 1 RCIM_Cubic>]}}}
     Scale Curve=2
    )
    Module.Index0Float0Write=true
    Module.Index0Float1Write=true
    Module.Index0Float2Write=true
    Module.Index0Float3Write=true
   )
   &ScaleColor(
    Scale Alpha=&FloatFromCurve(FloatCurve=DataInterface() {propertyValues {Curve {
        defaultValue=1
        keys [
         {arriveTangent=18.2055, interpMode=RCIM_Cubic, leaveTangent=18.2055, tangentMode=RCTM_User, time=0, value=0}
         {
          arriveTangent=-1.397
          arriveTangentWeight=0.1333
          interpMode=RCIM_Cubic
          leaveTangent=-1.397
          leaveTangentWeight=0.0667
          tangentMode=RCTM_User
          time=0.1
          value=1
         }
         {arriveTangent=-0.4376, tangentMode=RCTM_User, time=1, value=0}
        ]
       }}})
    Scale RGB=&Lerp_Vector(
     A=<1 1 1>
     B=<10 15 2>
     Alpha=&FloatFromCurve(
      FloatCurve=DataInterface() {propertyValues {Curve {keys [{arriveTangent=-1.3369, interpMode=RCIM_Cubic, leaveTangent=-1.3369, tangentMode=RCTM_User, time=0, value=1}, <at 1.25 0>]}}}
      CurveIndex=$Emitter.Age
     )
    )
    ScaleA=true
    ScaleRGB=true
   )
  ]
  rendererClasses [&NiagaraMeshRendererProperties]
  renderers [{rendererClass=&NiagaraMeshRendererProperties, rendererIndex=0}]
 }
}
summary {
 systemName=NS_DragonCart_Shard_Heart_Burst
 emitters [
  {emitterName=Minimal, rendererClasses [&NiagaraSpriteRendererProperties, &NiagaraLightRendererProperties]}
  {emitterName=RefractionSphere, rendererClasses [&NiagaraMeshRendererProperties]}
  {emitterName=PlasmaWave, rendererClasses []}
  {emitterName=Splines, rendererClasses [&NiagaraMeshRendererProperties]}
  {emitterName=Sparkles, simTarget=GPUComputeSim, rendererClasses [&NiagaraSpriteRendererProperties]}
 ]
 userVariables [User.Intensity {type=NiagaraFloat, defaultValue=1, description=""}]
}
