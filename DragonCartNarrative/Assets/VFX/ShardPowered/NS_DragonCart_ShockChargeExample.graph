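# Niagara example system for a DragonCart shock charge-up effect, driven by User.ChargeUp, with both emitters disabled in the asset.
# Minimal bursts a bright, persistent sprite that scales with charge and is offset toward the camera.
# Minimal001 spawns sprites on a sphere, or hemisphere via User.Hemisphere, with charge setting rate, radius and lifetime.
# These grow while curve-lerping position into the owner, with velocity recalculated from that motion.
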NiagaraSystem /DragonCartNarrative/Assets/VFX/ShardPowered/NS_DragonCart_ShockChargeExample
dependencies {
 usedDataInterfaces [&NiagaraDataInterfaceCurve, &NiagaraDataInterfaceVector2DCurve]
 usedDynamicInputs []
 usedModules [&SystemState]
}
emitters {
 Minimal {
  bEnabled=false
  emitterUpdateScript [
   &EmitterState(
    Inactive Response=::NewEnumerator1
    Life Cycle Mode=::NewEnumerator1
    Loop Behavior=::NewEnumerator1
    Loop Duration Mode=::NewEnumerator1
    UseLoopDelay=false
   )
   &SpawnBurst_Instantaneous(Use Spawn Probability=false, UseLoopCountLimit=false)
  ]
  particleSpawnScript [&InitializeParticle(Color=<rgba 10 10 10 1>, Sprite Size Mode=::NewEnumerator3, Uniform Sprite Size=1000, UsePositionOffset=false, Write Material Random=true)]
  particleUpdateScript [
   &ParticleState(Kill Particles When Lifetime Has Elapsed=false)
   &ScaleSpriteSize(Scale Sprite Size Mode=::NewEnumerator0, Uniform Scale Factor=$User.ChargeUp)
   &CameraOffset(Camera Offset Amount=/Niagara/DynamicInputs/Math/CalculateParticleRadius(
     Radius Calculation Type:ENiagaraCollisionRadiusOptions=::NewEnumerator0
     Particle Radius:NiagaraFloat=1
     Mesh Dimensions:Vector3f=<1 1 1>
     Method for Calculating Particle Radius:ENiagaraCalculateRadiusOptions=::NewEnumerator2
    ))
  ]
  rendererClasses [&NiagaraSpriteRendererProperties]
  renderers [{rendererClass=&NiagaraSpriteRendererProperties}]
 }
 Minimal001 {
  bEnabled=false
  emitterUpdateScript [
   &EmitterState()
   &SpawnRate(SpawnRate=&Multiply_Float(A=$User.ChargeUp, B=200), Use Spawn Probability=false)
  ]
  particleSpawnScript [
   &InitializeParticle(
    Color Mode=::NewEnumerator0
    Lifetime=&Lerp_Float(A=0.8, B=0.55, Alpha=$User.ChargeUp)
    Sprite Size Mode=::NewEnumerator3
    Uniform Sprite Size=50
    UsePositionOffset=false
    Write Material Random=true
   )
   &ShapeLocation(
    Hemisphere Distribution=&MakeVector2D(X=1, Y=&MakeCustomFloatFromBool(Bool=$User.Hemisphere, True Float=0.5, False Float=1))
    MaskSpawnGroup=false
    Override Seed=false
    Rotation Axis=<0 0 -1>
    Rotation Mode=::NewEnumerator1
    Sphere Radius=&Multiply_Float(A=2000, B=&Lerp_Float(A=0.5, B=1, Alpha=$User.ChargeUp))
    Sphere Surface Distribution=1
    UseOwnerScale=true
   )
  ]
  particleUpdateScript [
   &ParticleState()
   &ScaleSpriteSize(
    Non-Uniform Curve Scale=<1 6>
    Non-Uniform Curve Sprite Scale=DataInterface() {propertyValues {
      XCurve {keys [
        {arriveTangent=0.3117, interpMode=RCIM_Cubic, leaveTangent=0.3117, tangentMode=RCTM_User, time=0, value=0.25}
        {arriveTangent=1.5, interpMode=RCIM_Cubic, leaveTangent=1.5, tangentMode=RCTM_User, time=1, value=1}
       ]}
      YCurve {keys [
        {arriveTangent=0.5374, interpMode=RCIM_Cubic, leaveTangent=0.5374, tangentMode=RCTM_User, time=0, value=0.1}
        {arriveTangent=1.7616, interpMode=RCIM_Cubic, leaveTangent=1.7616, tangentMode=RCTM_User, time=1, value=1}
       ]}
     }}
    Scale Sprite Size Mode=::NewEnumerator3
   )
   /SetVariables(Particles.Position=&Lerp_Vector(
     A=Linked() {linkedVariable {name=Particles.Initial.Position, type=NiagaraPosition}}
     B=Linked() {linkedVariable {name=Engine.Owner.Position, type=NiagaraPosition}}
     Alpha=&FloatFromCurve(FloatCurve=DataInterface() {propertyValues {Curve {keys [
          {arriveTangent=0.1307, interpMode=RCIM_Cubic, leaveTangent=0.1307, tangentMode=RCTM_User, time=0, value=0}
          {arriveTangent=1.3788, tangentMode=RCTM_User, time=1, value=1}
         ]}}})
    )) {isSetParameters=true}
   &CalculateAccurateVelocity()
  ]
  rendererClasses [&NiagaraSpriteRendererProperties]
  renderers [{rendererClass=&NiagaraSpriteRendererProperties}]
 }
}
summary {
 systemName=NS_DragonCart_ShockChargeExample
 emitters [
  {bEnabled=false, emitterName=Minimal001, rendererClasses [&NiagaraSpriteRendererProperties]}
  {bEnabled=false, emitterName=Minimal, rendererClasses [&NiagaraSpriteRendererProperties]}
 ]
 userVariables [
  User.ChargeUp {type=NiagaraFloat, defaultValue=1, description=""}
  User.Hemisphere {type=NiagaraBool, defaultValue=true}
 ]
}
