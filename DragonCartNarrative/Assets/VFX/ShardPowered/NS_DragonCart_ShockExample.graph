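# Niagara example shock effect for Dragon Cart, pairing a flickering beam with a refraction shockwave; both emitters are saved disabled.
# Minimal stretches sprites between the owner and User.ZeroPointPosition, with waveform-driven width and glow size scaled by User.Intensity.
# User.DoShcokwave animates beam growth and glow flash, and gates a delayed RefractionSphere burst at the zero point.
# That mesh expands and fades through dynamic material parameters, one attenuated by camera distance.
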NiagaraSystem /DragonCartNarrative/Assets/VFX/ShardPowered/NS_DragonCart_ShockExample
dependencies {
 usedDataInterfaces [&NiagaraDataInterfaceCurve, &NiagaraDataInterfaceColorCurve]
 usedDynamicInputs []
 usedModules [&SystemState]
}
emitters {
 Minimal {
  bEnabled=false
  emitterUpdateScript [
   &EmitterState(
    Inactive Response=::NewEnumerator1
    Life Cycle Mode=::NewEnumerator1
    Loop Behavior=::NewEnumerator1
    Loop Duration Mode=::NewEnumerator1
    UseLoopDelay=false
   )
   &SpawnBurst_Instantaneous(Use Spawn Probability=false, UseLoopCountLimit=false)
  ]
  particleSpawnScript [
   &InitializeParticle(
    Color Mode=::NewEnumerator0
    Lifetime=5
    Mass=700
    Mass Mode=::NewEnumerator1
    Sprite Size Mode=::NewEnumerator3
    Uniform Sprite Size=1
    UsePositionOffset=false
    Write Material Random=true
   )
   &SpriteFacingAndAlignment(Sprite Alignment=<0 0 1>, WriteAlignment=true, WriteFacing=false)
  ]
  particleUpdateScript [
   &ParticleState(Kill Particles When Lifetime Has Elapsed=false)
   /SetVariables(Particles.GlowPosition:NiagaraPosition=&AddVectorToPosition(Position=$Engine.Owner.Position, Vector=<0 0 -200>)) {isSetParameters=true}
   /SetVariables(Particles.Position=&Lerp_Vector(
     A=Linked() {linkedVariable {name=Engine.Owner.Position, type=NiagaraPosition}}
     B=Linked() {linkedVariable {name=User.ZeroPointPosition, type=NiagaraPosition}}
     Alpha=&MakeCustomFloatFromBool(
      Bool=$User.DoShcokwave
      True Float=&FloatFromCurve(FloatCurve=DataInterface() {propertyValues {Curve {keys [<at 0 0>,<at 0.2 0.5>]}}}, CurveIndex=$Particles.Age)
      False Float=0.5
     )
    )) {isSetParameters=true}
   /SetVariables(Particles.BeamWidth:NiagaraFloat=&Waveform(
     Global Amplitude Scale=&Multiply_Float(A=$User.Intensity, B=2000)
     "Waveform [1]"=::NewEnumerator0
     "[1] Frequency"=&Lerp_Float(A=7, B=9, Alpha=$User.Intensity)
     "[1] Amplitude Min/Max"=<0.8 1>
    )) {isSetParameters=true}
   /SetVariables(Particles.GlowSize=&Vector2DFromFloat(Value=&Add_Float(
      A=&Multiply_Float(
       A=$Particles.BeamWidth
       B=&MakeCustomFloatFromBool(
        Bool=$User.DoShcokwave
        True Float=&FloatFromCurve(
         FloatCurve=DataInterface() {propertyValues {Curve {keys [
             {arriveTangent=80.2797, interpMode=RCIM_Cubic, leaveTangent=80.2797, tangentMode=RCTM_User, time=0, value=0}
             {arriveTangent=-8.6948, interpMode=RCIM_Cubic, leaveTangent=-8.6948, tangentMode=RCTM_User, time=0.1, value=2.5}
             <at 0.5 1>
            ]}}}
         CurveIndex=$Particles.Age
         Scale Curve=2
        )
        False Float=2
       )
      )
      B=0
     ))) {isSetParameters=true}
   &ScaleSpriteSize(
    Scale Factor=&MakeVector2D(
     X=$Particles.BeamWidth
     Y=&Add_Float(
      A=&Multiply_Float(A=&DistanceBetweenPositions(A=$Engine.Owner.Position, B=$Particles.Position), B=2)
      B=$Particles.BeamWidth
     )
    )
    Scale Sprite Size Mode=::NewEnumerator2
   )
   &ScaleColor(
    Scale RGB=&VectorFromFloat(Value=&Lerp_Float(A=0.25, B=3, Alpha=$User.Intensity))
    ScaleA=true
    ScaleRGB=true
   )
  ]
  rendererClasses [&NiagaraSpriteRendererProperties]
  renderers [
   {rendererClass=&NiagaraSpriteRendererProperties}
   {rendererClass=&NiagaraSpriteRendererProperties, rendererIndex=1}
  ]
 }
 RefractionSphere {
  bEnabled=false
  emitterUpdateScript [
   &EmitterState(Life Cycle Mode=::NewEnumerator1, Loop Behavior=::NewEnumerator1, Loop Duration=1, UseLoopDelay=false)
   &SpawnBurst_Instantaneous(
    Spawn Count=&MakeCustomIntFromBool(Bool=$User.DoShcokwave, True INT=1, False INT=0)
    Spawn Time=0.15
    Use Spawn Probability=false
    UseLoopCountLimit=false
   )
  ]
  particleSpawnScript [
   &InitializeParticle(
    Color=<rgba 2 2 2 0.75>
    Lifetime=&Lerp_Float(A=0.7, B=1.5, Alpha=$User.Intensity)
    Mesh Scale Mode=::NewEnumerator3
    Mesh Uniform Scale=&Multiply_Float(A=$User.Intensity, B=120)
    Position=$User.ZeroPointPosition
    Position Mode=::NewEnumerator1
    UsePositionOffset=false
    Write Material Random=true
   )
   &InitialMeshOrientation(Mesh Axis To Orient=<0 0 1>, Mesh Orientation Mode=::NewEnumerator2, Use Rotation Vector=false)
  ]
  particleUpdateScript [
   &ParticleState()
   &ScaleMeshSize(Scale Factor=&VectorFromFloat(Value=&FloatFromCurve(FloatCurve=DataInterface() {propertyValues {Curve {defaultValue=0, keys [{tangentMode=RCTM_User, time=0, value=0}, {tangentMode=RCTM_User, time=1, value=1}]}}})))
   &ScaleColor(
    Linear Color Curve=DataInterface() {propertyValues {
      AlphaCurve {keys [<at 0.4 1 RCIM_Cubic>, {arriveTangent=-2.2959, tangentMode=RCTM_User, time=1, value=0}]}
      BlueCurve {keys [<at 0 2 RCIM_Cubic>]}
      GreenCurve {keys [<at 0 2 RCIM_Cubic>]}
      RedCurve {keys [<at 0 2 RCIM_Cubic>]}
     }}
    Scale Mode=::NewEnumerator2
   )
   &DynamicMaterialParameters(
    Index 0 Param 1=&Lerp_Float(
     A=1
     B=0.87
     Alpha=&FloatFromCurve(
      FloatCurve=DataInterface() {propertyValues {Curve {
         defaultValue=1
         keys [
          {interpMode=RCIM_Cubic, leaveTangentWeight=0.0333, tangentMode=RCTM_User, time=0, value=0}
          {
           arriveTangent=-1.6584
           arriveTangentWeight=0.0646
           interpMode=RCIM_Cubic
           leaveTangent=-1.6584
           leaveTangentWeight=0.581
           tangentMode=RCTM_User
           time=0.1
           value=1
          }
          {arriveTangentWeight=0.3, interpMode=RCIM_Cubic, tangentMode=RCTM_User, time=1, value=0}
         ]
        }}}
      Scale Curve=&FN_LOD_Distance_Remap(Start Distance=5000, End Distance=20000, Minimum Output=1, Maximum Output=0, Clamp Output=true)
     )
    )
    Index 0 Param 2=&FloatFromCurve(FloatCurve=DataInterface() {propertyValues {Curve {
        defaultValue=0
        keys [
         {arriveTangent=-2.4203, interpMode=RCIM_Cubic, time=0, value=1}
         {arriveTangent=-1.765, interpMode=RCIM_Cubic, leaveTangent=-1.765, tangentMode=RCTM_User, time=1, value=0}
        ]
       }}})
    Index 0 Param 3=&Lerp_Float(
     A=1
     B=0.5
     Alpha=&FloatFromCurve(FloatCurve=DataInterface() {propertyValues {Curve {defaultValue=1, keys [{arriveTangent=1.7704, interpMode=RCIM_Cubic, time=0, value=0}, {tangentMode=RCTM_User, time=1, value=1}]}}})
    )
    Index 0 Param 4=2000
    Module.Index0Float0Write=true
    Module.Index0Float1Write=true
    Module.Index0Float2Write=true
    Module.Index0Float3Write=true
   )
  ]
  rendererClasses [&NiagaraMeshRendererProperties]
  renderers [{rendererClass=&NiagaraMeshRendererProperties, rendererIndex=0}]
 }
}
summary {
 systemName=NS_DragonCart_ShockExample
 emitters [
  {bEnabled=false, emitterName=Minimal, rendererClasses [&NiagaraSpriteRendererProperties]}
  {bEnabled=false, emitterName=RefractionSphere, rendererClasses [&NiagaraMeshRendererProperties]}
 ]
 userVariables [
  User.DoShcokwave {type=NiagaraBool, defaultValue=true}
  User.Intensity {type=NiagaraFloat, defaultValue=1, description=""}
  User.ZeroPointPosition {type=NiagaraPosition, description="", defaultValue {}}
 ]
}
